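# Physics asset that simulates the skirt on the PureCereal medium male player body.
# Five-link capsule chains (front, side, rear, corners) hang from the pelvis with limited swing and twist, except free-rotating corner roots and tiny rear _MSTR intermediate bodies.
# Free-angle cross-constraints tie lower front and rear panels together.
# Pelvis and angularly locked leg capsules provide body shapes, plus a large sphere below dyn_root_offset, apparently as a floor.
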PhysicsAsset /BRCosmetics/Characters/Player/Male/Medium/Bodies/M_MED_PureCereal/Meshes/M_MED_PureCereal_Skirt_Physics
bodies {
 calf_l [{shapeType=Capsule, length=39.5729, radius=9.3359, center=<-24.3994 0.7239 -0.3023>, rotation=<rot 0 -91.0756 -90.1505>}]
 calf_r [{shapeType=Capsule, length=39.5729, radius=9.3359, center=<24.3994 -0.7239 0.3023>, rotation=<rot 0 -91.0756 90.1505>}]
 dyn_front_skirt_corner_l_01 [{
   shapeType=Capsule
   length=8.8414
   radius=3.1608
   shapeName=dyn_front_skirt_corner_l_01_capsule
   center=<6.808 0.5688 -0.6319>
   rotation=<rot 0 -94.9519 -80.1226>
  }]
 dyn_front_skirt_corner_l_02 [{shapeType=Capsule, length=17.4572, radius=8.5428, shapeName=dyn_front_skirt_corner_l_02_capsule, center=<9 -6.7967 0>}]
 dyn_front_skirt_corner_l_03 [{shapeType=Capsule, length=20, radius=8, shapeName=dyn_front_skirt_corner_l_03_capsule, center=<4 -8 0>}]
 dyn_front_skirt_corner_l_04 [{shapeType=Capsule, length=20, radius=8, shapeName=dyn_front_skirt_corner_l_04_capsule, center=<4 -8 0>}]
 dyn_front_skirt_corner_l_05 [{shapeType=Capsule, length=20, radius=9.9, shapeName=dyn_front_skirt_corner_l_05_capsule, center=<4 -8 0>}]
 dyn_front_skirt_corner_r_01 [{
   shapeType=Capsule
   length=10.9013
   radius=3.6976
   shapeName=dyn_front_skirt_corner_r_01_capsule
   center=<-9.3602 -0.7459 -0.5138>
   rotation=<rot 0 89.8025 -89.3441>
  }]
 dyn_front_skirt_corner_r_02 [{shapeType=Capsule, length=20, radius=8.5428, shapeName=dyn_front_skirt_corner_r_02_capsule, center=<-9 6.7708 0>}]
 dyn_front_skirt_corner_r_03 [{shapeType=Capsule, length=20, radius=8, shapeName=dyn_front_skirt_corner_r_03_capsule, center=<-4 8 0>}]
 dyn_front_skirt_corner_r_04 [{shapeType=Capsule, length=20, radius=8, shapeName=dyn_front_skirt_corner_r_04_capsule, center=<-4 8 0>}]
 dyn_front_skirt_corner_r_05 [{shapeType=Capsule, length=20, radius=9.9, shapeName=dyn_front_skirt_corner_r_05_capsule, center=<-4 8 0>}]
 dyn_front_skirt_l_01 [{shapeType=Capsule, length=11.3, radius=4.7, shapeName=dyn_front_skirt_l_01_capsule, center=<5 -6 0>}]
 dyn_front_skirt_l_02 [{shapeType=Capsule, length=5, radius=8.5428, shapeName=dyn_front_skirt_l_02_capsule, center=<9 -4 0>}]
 dyn_front_skirt_l_03 [{shapeType=Capsule, length=12.6906, radius=8, shapeName=dyn_front_skirt_l_03_capsule, center=<5 -6 0>}]
 dyn_front_skirt_l_04 [{shapeType=Capsule, length=12.6906, radius=8, shapeName=dyn_front_skirt_l_04_capsule, center=<5 -4 0>}]
 dyn_front_skirt_l_05 [{shapeType=Capsule, length=12.6906, radius=9.9, shapeName=dyn_front_skirt_l_05_capsule, center=<5 -4 0>}]
 dyn_front_skirt_r_01 [{shapeType=Capsule, length=11.3, radius=4.7, shapeName=dyn_front_skirt_l_01_capsule, center=<-5 6 0>}]
 dyn_front_skirt_r_02 [{shapeType=Capsule, length=5, radius=8.5428, shapeName=dyn_front_skirt_l_02_capsule, center=<-9 4 0>, rotation=<rot 0 0 -180>}]
 dyn_front_skirt_r_03 [{shapeType=Capsule, length=12.6906, radius=8, shapeName=dyn_front_skirt_l_03_capsule, center=<-5 6 0>, rotation=<rot 0 0 -180>}]
 dyn_front_skirt_r_04 [{shapeType=Capsule, length=12.6906, radius=8, shapeName=dyn_front_skirt_l_04_capsule, center=<-5 4 0>, rotation=<rot 0 0 -180>}]
 dyn_front_skirt_r_05 [{shapeType=Capsule, length=12.6906, radius=9.9, shapeName=dyn_front_skirt_l_05_capsule, center=<-5 4 0>, rotation=<rot 0 0 -180>}]
 dyn_rear_skirt_corner_l_01 [{
   shapeType=Capsule
   length=11.1251
   radius=3.4951
   shapeName=dyn_rear_skirt_corner_l_01_capsule
   center=<9.822 0.0181 0.536>
   rotation=<rot 0 88.4386 -86.0857>
  }]
 dyn_rear_skirt_corner_l_02 [{shapeType=Capsule, length=14, radius=8.5428, shapeName=dyn_rear_skirt_corner_l_02_capsule, center=<9 -8 0>}]
 dyn_rear_skirt_corner_l_02_MSTR [{shapeType=Capsule, length=0.505, radius=0.505, shapeName=dyn_rear_skirt_corner_l_02_MSTR_capsule, rotation=<rot 0 -90 0>}]
 dyn_rear_skirt_corner_l_03 [{shapeType=Capsule, length=20, radius=8, shapeName=dyn_rear_skirt_corner_l_03_capsule, center=<4 -8 0>}]
 dyn_rear_skirt_corner_l_04 [{shapeType=Capsule, length=30, radius=8, shapeName=dyn_rear_skirt_corner_l_04_capsule, center=<4 -8.1699 0>}]
 dyn_rear_skirt_corner_l_05 [{shapeType=Capsule, length=28.4, radius=11.5, shapeName=dyn_rear_skirt_corner_l_05_capsule, center=<8 -8 0>}]
 dyn_rear_skirt_corner_r_01 [{
   shapeType=Capsule
   length=11.4719
   radius=4.6792
   shapeName=dyn_rear_skirt_corner_r_01_capsule
   center=<-9.7011 0.4736 -0.3033>
   rotation=<rot 0 88.486 -81.7393>
  }]
 dyn_rear_skirt_corner_r_02 [{shapeType=Capsule, length=14, radius=8.5428, shapeName=dyn_rear_skirt_corner_r_02_capsule, center=<-9 8 0>}]
 dyn_rear_skirt_corner_r_02_MSTR [{shapeType=Capsule, length=0.505, radius=0.505, shapeName=dyn_rear_skirt_corner_r_02_MSTR_capsule, rotation=<rot 0 -90 0>}]
 dyn_rear_skirt_corner_r_03 [{shapeType=Capsule, length=20, radius=8, shapeName=dyn_rear_skirt_corner_r_03_capsule, center=<-4 8 0>}]
 dyn_rear_skirt_corner_r_04 [{shapeType=Capsule, length=30, radius=8, shapeName=dyn_rear_skirt_corner_r_04_capsule, center=<-4 7.8301 0>}]
 dyn_rear_skirt_corner_r_05 [{shapeType=Capsule, length=28.4, radius=11.5, shapeName=dyn_rear_skirt_corner_r_05_capsule, center=<-8 8 0>}]
 dyn_rear_skirt_l_01 [{shapeType=Capsule, length=11.3, radius=4.7, shapeName=dyn_rear_skirt_l_01_capsule, center=<5 -6 0>}]
 dyn_rear_skirt_l_02 [{shapeType=Capsule, length=11.9, radius=8.5428, shapeName=dyn_rear_skirt_r_02_capsule, center=<9 -4.1708 0>, rotation=<rot 0 0 -180>}]
 dyn_rear_skirt_l_02_MSTR [{shapeType=Capsule, length=0.505, radius=0.505, shapeName=dyn_rear_skirt_l_02_MSTR_capsule, rotation=<rot 0 -90 0>}]
 dyn_rear_skirt_l_03 [{shapeType=Capsule, length=20, radius=8, shapeName=dyn_rear_skirt_r_04_capsule, center=<5 -4 0>, rotation=<rot 0 0 -180>}]
 dyn_rear_skirt_l_04 [{shapeType=Capsule, length=30, radius=8, shapeName=dyn_side_skirt_l_MSTR_capsule, center=<5 -3.8301 0>, rotation=<rot 0 0 -180>}]
 dyn_rear_skirt_l_05 [{shapeType=Capsule, length=28.4, radius=11.5, shapeName=dyn_side_skirt_l_02_capsule, center=<8 -4 0>, rotation=<rot 0 0 -180>}]
 dyn_rear_skirt_r_01 [{shapeType=Capsule, length=11.3, radius=4.7, shapeName=dyn_rear_skirt_r_MSTR_capsule, center=<-5 6 0>}]
 dyn_rear_skirt_r_02 [{shapeType=Capsule, length=11.9, radius=8.5428, shapeName=dyn_rear_skirt_r_02_capsule, center=<-9 4 0>}]
 dyn_rear_skirt_r_02_MSTR [{shapeType=Capsule, length=0.505, radius=0.505, shapeName=dyn_rear_skirt_r_02_MSTR_capsule, rotation=<rot 0 -90 0>}]
 dyn_rear_skirt_r_03 [{shapeType=Capsule, length=20, radius=8, shapeName=dyn_rear_skirt_r_04_capsule, center=<-5 4 0>}]
 dyn_rear_skirt_r_04 [{shapeType=Capsule, length=30, radius=8, shapeName=dyn_side_skirt_l_MSTR_capsule, center=<-5 3.8301 0>}]
 dyn_rear_skirt_r_05 [{shapeType=Capsule, length=28.4, radius=11.5, shapeName=dyn_side_skirt_l_02_capsule, center=<-8 4 0>}]
 dyn_root_offset [{shapeType=Sphere, radius=300, shapeName=dyn_root_offset_sphere, center=<0 0 -300>}]
 dyn_side_skirt_l_01 [{shapeType=Capsule, length=11.3, radius=4.7, shapeName=dyn_side_skirt_l_01_capsule, center=<5 -2.4075 0>}]
 dyn_side_skirt_l_02 [{shapeType=Capsule, length=14, radius=8.5428, shapeName=dyn_side_skirt_l_02_capsule, center=<9 -8.0402 0>}]
 dyn_side_skirt_l_03 [{shapeType=Capsule, length=29.8, radius=9.1, shapeName=dyn_side_skirt_l_03_capsule, center=<5 -11.6635 0>}]
 dyn_side_skirt_l_04 [{shapeType=Capsule, length=39.8, radius=9.1, shapeName=dyn_side_skirt_l_04_capsule, center=<5 -11.8314 0>}]
 dyn_side_skirt_l_05 [{
   shapeType=Capsule
   length=37.9001
   radius=12.9
   shapeName=dyn_side_skirt_l_05_capsule
   center=<5 -13.5737 -2.9751>
   rotation=<rot 0.0001 0 -8.4375>
  }]
 dyn_side_skirt_r_01 [{shapeType=Capsule, length=11.3, radius=4.7, shapeName=dyn_side_skirt_r_05_capsule, center=<-5 2.4711 0>}]
 dyn_side_skirt_r_02 [{shapeType=Capsule, length=14, radius=8.5428, shapeName=dyn_side_skirt_l_02_capsule, center=<-9 8 0>, rotation=<rot 0.0003 0 177.1875>}]
 dyn_side_skirt_r_03 [{
   shapeType=Capsule
   length=30.5
   radius=8.4
   shapeName=dyn_side_skirt_l_03_capsule
   center=<-5 11.037 0.0961>
   rotation=<rot 0.0003 0 177.1875>
  }]
 dyn_side_skirt_r_04 [{
   shapeType=Capsule
   length=39.9
   radius=9
   shapeName=dyn_side_skirt_l_04_capsule
   center=<-5 11.1938 0.1687>
   rotation=<rot 0.0004 0 -177.1875>
  }]
 dyn_side_skirt_r_05 [{shapeType=Capsule, length=38.6001, radius=12.2, shapeName=dyn_side_skirt_l_05_capsule, center=<-5 13.0339 0>, rotation=<rot 0 0 180>}]
 foot_l [{shapeType=Capsule, length=15.0406, radius=10.4054, center=<-5.5143 6.6493 -1.684>, rotation=<rot -0.3035 22.8005 -88.4741>}]
 foot_r [{shapeType=Capsule, length=15.0406, radius=10.4054, center=<5.5143 -6.6493 1.684>, rotation=<rot -0.3035 22.8005 91.5259>}]
 pelvis [{shapeType=Capsule, length=10, radius=12, rotation=<rot 0 90 0>}]
 thigh_l [{shapeType=Capsule, length=49.7, radius=8.3139, center=<-19.4923 -0.8591 0.5803>, rotation=<rot 1.1057 -88.9384 -89.8161>}]
 thigh_r [{shapeType=Capsule, length=49.7, radius=8.3139, center=<19.4923 0.8591 -0.5803>, rotation=<rot 1.1057 -88.9385 90.1839>}]
}
constraints [
 {
  bone1Name=thigh_l
  bone2Name=pelvis
  swing1LimitDegrees=45
  swing1Motion=Locked
  swing2LimitDegrees=45
  swing2Motion=Locked
  twistLimitDegrees=45
  twistMotion=Locked
 }
 {
  bone1Name=calf_l
  bone2Name=thigh_l
  swing1LimitDegrees=45
  swing1Motion=Locked
  swing2LimitDegrees=45
  swing2Motion=Locked
  twistLimitDegrees=45
  twistMotion=Locked
 }
 {
  bone1Name=foot_l
  bone2Name=calf_l
  swing1LimitDegrees=45
  swing1Motion=Locked
  swing2LimitDegrees=45
  swing2Motion=Locked
  twistLimitDegrees=45
  twistMotion=Locked
 }
 {bone1Name=dyn_front_skirt_l_01, bone2Name=pelvis, swing1LimitDegrees=32, swing2LimitDegrees=12, twistLimitDegrees=12}
 {bone1Name=dyn_front_skirt_l_02, bone2Name=dyn_front_skirt_l_01, swing1LimitDegrees=8, swing2LimitDegrees=6, twistLimitDegrees=6}
 {bone1Name=dyn_front_skirt_l_03, bone2Name=dyn_front_skirt_l_02, swing1LimitDegrees=32, swing2LimitDegrees=24, twistLimitDegrees=24}
 {bone1Name=dyn_front_skirt_l_04, bone2Name=dyn_front_skirt_l_03, swing1LimitDegrees=32, swing2LimitDegrees=24, twistLimitDegrees=24}
 {bone1Name=dyn_front_skirt_l_05, bone2Name=dyn_front_skirt_l_04, swing1LimitDegrees=12, swing2LimitDegrees=4, twistLimitDegrees=4}
 {bone1Name=dyn_rear_skirt_l_01, bone2Name=pelvis, swing1LimitDegrees=32, swing2LimitDegrees=12, twistLimitDegrees=12}
 {bone1Name=dyn_side_skirt_l_01, bone2Name=pelvis, swing1LimitDegrees=32, swing2LimitDegrees=12, twistLimitDegrees=12}
 {bone1Name=dyn_side_skirt_l_02, bone2Name=dyn_side_skirt_l_01, swing1LimitDegrees=8, swing2LimitDegrees=14, twistLimitDegrees=14}
 {bone1Name=dyn_side_skirt_l_03, bone2Name=dyn_side_skirt_l_02, swing1LimitDegrees=32, swing2LimitDegrees=24, twistLimitDegrees=24}
 {bone1Name=dyn_side_skirt_l_04, bone2Name=dyn_side_skirt_l_03, swing1LimitDegrees=32, swing2LimitDegrees=24, twistLimitDegrees=24}
 {bone1Name=dyn_side_skirt_l_05, bone2Name=dyn_side_skirt_l_04, swing1LimitDegrees=12, swing2LimitDegrees=4, twistLimitDegrees=4}
 {bone1Name=dyn_front_skirt_r_01, bone2Name=pelvis, swing1LimitDegrees=32, swing2LimitDegrees=12, twistLimitDegrees=12}
 {bone1Name=dyn_rear_skirt_r_01, bone2Name=pelvis, swing1LimitDegrees=32, swing2LimitDegrees=12, twistLimitDegrees=12}
 {bone1Name=dyn_rear_skirt_r_02, bone2Name=dyn_rear_skirt_r_02_MSTR, swing1LimitDegrees=24, swing2LimitDegrees=8, twistLimitDegrees=8}
 {bone1Name=dyn_rear_skirt_r_03, bone2Name=dyn_rear_skirt_r_02, swing1LimitDegrees=24, swing2LimitDegrees=10, twistLimitDegrees=8}
 {bone1Name=dyn_rear_skirt_r_04, bone2Name=dyn_rear_skirt_r_03, swing1LimitDegrees=24, swing2LimitDegrees=10, twistLimitDegrees=8}
 {bone1Name=dyn_rear_skirt_r_05, bone2Name=dyn_rear_skirt_r_04, swing1LimitDegrees=12, swing2LimitDegrees=10, twistLimitDegrees=10}
 {bone1Name=dyn_side_skirt_r_01, bone2Name=pelvis, swing1LimitDegrees=32, swing2LimitDegrees=12, twistLimitDegrees=12}
 {
  bone1Name=thigh_r
  bone2Name=pelvis
  swing1LimitDegrees=45
  swing1Motion=Locked
  swing2LimitDegrees=45
  swing2Motion=Locked
  twistLimitDegrees=45
  twistMotion=Locked
 }
 {bone1Name=dyn_front_skirt_r_02, bone2Name=dyn_front_skirt_r_01, swing1LimitDegrees=8, swing2LimitDegrees=6, twistLimitDegrees=6}
 {bone1Name=dyn_front_skirt_r_03, bone2Name=dyn_front_skirt_r_02, swing1LimitDegrees=32, swing2LimitDegrees=24, twistLimitDegrees=24}
 {bone1Name=dyn_front_skirt_r_04, bone2Name=dyn_front_skirt_r_03, swing1LimitDegrees=32, swing2LimitDegrees=24, twistLimitDegrees=24}
 {bone1Name=dyn_front_skirt_r_05, bone2Name=dyn_front_skirt_r_04, swing1LimitDegrees=12, swing2LimitDegrees=4, twistLimitDegrees=4}
 {bone1Name=dyn_side_skirt_r_02, bone2Name=dyn_side_skirt_r_01, swing1LimitDegrees=8, swing2LimitDegrees=14, twistLimitDegrees=14}
 {bone1Name=dyn_side_skirt_r_03, bone2Name=dyn_side_skirt_r_02, swing1LimitDegrees=32, swing2LimitDegrees=24, twistLimitDegrees=24}
 {bone1Name=dyn_side_skirt_r_04, bone2Name=dyn_side_skirt_r_03, swing1LimitDegrees=32, swing2LimitDegrees=24, twistLimitDegrees=24}
 {bone1Name=dyn_side_skirt_r_05, bone2Name=dyn_side_skirt_r_04, swing1LimitDegrees=12, swing2LimitDegrees=4, twistLimitDegrees=4}
 {
  bone1Name=dyn_front_skirt_corner_l_01
  bone2Name=pelvis
  swing1LimitDegrees=45
  swing1Motion=Free
  swing2LimitDegrees=45
  swing2Motion=Free
  twistLimitDegrees=45
  twistMotion=Free
 }
 {
  bone1Name=dyn_front_skirt_corner_l_02
  bone2Name=dyn_front_skirt_corner_l_01
  swing1LimitDegrees=8
  swing2LimitDegrees=6
  twistLimitDegrees=6
 }
 {
  bone1Name=dyn_front_skirt_corner_l_03
  bone2Name=dyn_front_skirt_corner_l_02
  swing1LimitDegrees=32
  swing2LimitDegrees=24
  twistLimitDegrees=24
 }
 {
  bone1Name=dyn_front_skirt_corner_l_04
  bone2Name=dyn_front_skirt_corner_l_03
  swing1LimitDegrees=32
  swing2LimitDegrees=24
  twistLimitDegrees=24
 }
 {
  bone1Name=dyn_front_skirt_corner_l_05
  bone2Name=dyn_front_skirt_corner_l_04
  swing1LimitDegrees=12
  swing2LimitDegrees=4
  twistLimitDegrees=4
 }
 {
  bone1Name=dyn_front_skirt_corner_r_01
  bone2Name=pelvis
  swing1LimitDegrees=45
  swing1Motion=Free
  swing2LimitDegrees=45
  swing2Motion=Free
  twistLimitDegrees=45
  twistMotion=Free
 }
 {
  bone1Name=dyn_front_skirt_corner_r_02
  bone2Name=dyn_front_skirt_corner_r_01
  swing1LimitDegrees=8
  swing2LimitDegrees=6
  twistLimitDegrees=6
 }
 {
  bone1Name=dyn_front_skirt_corner_r_03
  bone2Name=dyn_front_skirt_corner_r_02
  swing1LimitDegrees=32
  swing2LimitDegrees=24
  twistLimitDegrees=24
 }
 {
  bone1Name=dyn_front_skirt_corner_r_04
  bone2Name=dyn_front_skirt_corner_r_03
  swing1LimitDegrees=32
  swing2LimitDegrees=24
  twistLimitDegrees=24
 }
 {
  bone1Name=dyn_front_skirt_corner_r_05
  bone2Name=dyn_front_skirt_corner_r_04
  swing1LimitDegrees=12
  swing2LimitDegrees=4
  twistLimitDegrees=4
 }
 {
  bone1Name=dyn_rear_skirt_corner_l_01
  bone2Name=pelvis
  swing1LimitDegrees=45
  swing1Motion=Free
  swing2LimitDegrees=45
  swing2Motion=Free
  twistLimitDegrees=45
  twistMotion=Free
 }
 {
  bone1Name=dyn_rear_skirt_corner_l_02
  bone2Name=dyn_rear_skirt_corner_l_02_MSTR
  swing1LimitDegrees=16
  swing2LimitDegrees=4
  twistLimitDegrees=4
 }
 {
  bone1Name=dyn_rear_skirt_corner_l_03
  bone2Name=dyn_rear_skirt_corner_l_02
  swing1LimitDegrees=24
  swing2LimitDegrees=20
  twistLimitDegrees=12
 }
 {
  bone1Name=dyn_rear_skirt_corner_l_04
  bone2Name=dyn_rear_skirt_corner_l_03
  swing1LimitDegrees=24
  swing2LimitDegrees=20
  twistLimitDegrees=12
 }
 {
  bone1Name=dyn_rear_skirt_corner_l_05
  bone2Name=dyn_rear_skirt_corner_l_04
  swing1LimitDegrees=12
  swing2LimitDegrees=10
  twistLimitDegrees=10
 }
 {
  bone1Name=dyn_rear_skirt_corner_r_01
  bone2Name=pelvis
  swing1LimitDegrees=45
  swing1Motion=Free
  swing2LimitDegrees=45
  swing2Motion=Free
  twistLimitDegrees=45
  twistMotion=Free
 }
 {
  bone1Name=dyn_rear_skirt_corner_r_02
  bone2Name=dyn_rear_skirt_corner_r_02_MSTR
  swing1LimitDegrees=16
  swing2LimitDegrees=10
  twistLimitDegrees=10
 }
 {
  bone1Name=dyn_rear_skirt_corner_r_03
  bone2Name=dyn_rear_skirt_corner_r_02
  swing1LimitDegrees=24
  swing2LimitDegrees=20
  twistLimitDegrees=12
 }
 {
  bone1Name=dyn_rear_skirt_corner_r_04
  bone2Name=dyn_rear_skirt_corner_r_03
  swing1LimitDegrees=24
  swing2LimitDegrees=20
  twistLimitDegrees=12
 }
 {
  bone1Name=dyn_rear_skirt_corner_r_05
  bone2Name=dyn_rear_skirt_corner_r_04
  swing1LimitDegrees=12
  swing2LimitDegrees=10
  twistLimitDegrees=10
 }
 {
  bone1Name=dyn_front_skirt_l_05
  bone2Name=dyn_front_skirt_r_05
  swing1LimitDegrees=45
  swing1Motion=Free
  swing2LimitDegrees=45
  swing2Motion=Free
  twistLimitDegrees=45
  twistMotion=Free
 }
 {
  bone1Name=dyn_front_skirt_r_05
  bone2Name=dyn_front_skirt_l_05
  swing1LimitDegrees=45
  swing1Motion=Free
  swing2LimitDegrees=45
  swing2Motion=Free
  twistLimitDegrees=45
  twistMotion=Free
 }
 {
  bone1Name=calf_r
  bone2Name=thigh_r
  swing1LimitDegrees=45
  swing1Motion=Locked
  swing2LimitDegrees=45
  swing2Motion=Locked
  twistLimitDegrees=45
  twistMotion=Locked
 }
 {
  bone1Name=dyn_rear_skirt_corner_l_02_MSTR
  bone2Name=dyn_rear_skirt_corner_l_01
  swing1LimitDegrees=45
  swing1Motion=Free
  swing2LimitDegrees=45
  swing2Motion=Free
  twistLimitDegrees=45
  twistMotion=Free
 }
 {
  bone1Name=dyn_rear_skirt_corner_r_02_MSTR
  bone2Name=dyn_rear_skirt_corner_r_01
  swing1LimitDegrees=45
  swing1Motion=Free
  swing2LimitDegrees=45
  swing2Motion=Free
  twistLimitDegrees=45
  twistMotion=Free
 }
 {
  bone1Name=dyn_rear_skirt_l_02_MSTR
  bone2Name=dyn_rear_skirt_l_01
  swing1LimitDegrees=45
  swing1Motion=Free
  swing2LimitDegrees=45
  swing2Motion=Free
  twistLimitDegrees=45
  twistMotion=Free
 }
 {
  bone1Name=dyn_rear_skirt_r_02_MSTR
  bone2Name=dyn_rear_skirt_r_01
  swing1LimitDegrees=45
  swing1Motion=Free
  swing2LimitDegrees=45
  swing2Motion=Free
  twistLimitDegrees=45
  twistMotion=Free
 }
 {
  bone1Name=dyn_front_skirt_corner_r_05
  bone2Name=dyn_front_skirt_r_05
  swing1LimitDegrees=45
  swing1Motion=Free
  swing2LimitDegrees=45
  swing2Motion=Free
  twistLimitDegrees=45
  twistMotion=Free
 }
 {
  bone1Name=foot_r
  bone2Name=calf_r
  swing1LimitDegrees=45
  swing1Motion=Locked
  swing2LimitDegrees=45
  swing2Motion=Locked
  twistLimitDegrees=45
  twistMotion=Locked
 }
 {bone1Name=dyn_rear_skirt_l_02, bone2Name=dyn_rear_skirt_l_02_MSTR, swing1LimitDegrees=32, swing2LimitDegrees=24, twistLimitDegrees=24}
 {bone1Name=dyn_rear_skirt_l_03, bone2Name=dyn_rear_skirt_l_02, swing1LimitDegrees=24, swing2LimitDegrees=10, twistLimitDegrees=8}
 {bone1Name=dyn_rear_skirt_l_04, bone2Name=dyn_rear_skirt_l_03, swing1LimitDegrees=24, swing2LimitDegrees=10, twistLimitDegrees=8}
 {bone1Name=dyn_rear_skirt_l_05, bone2Name=dyn_rear_skirt_l_04, swing1LimitDegrees=12, swing2LimitDegrees=10, twistLimitDegrees=10}
 {
  bone1Name=dyn_rear_skirt_l_04
  bone2Name=dyn_rear_skirt_r_04
  swing1LimitDegrees=45
  swing1Motion=Free
  swing2LimitDegrees=45
  swing2Motion=Free
  twistLimitDegrees=45
  twistMotion=Free
 }
 {
  bone1Name=dyn_rear_skirt_r_04
  bone2Name=dyn_rear_skirt_l_04
  swing1LimitDegrees=45
  swing1Motion=Free
  swing2LimitDegrees=45
  swing2Motion=Free
  twistLimitDegrees=45
  twistMotion=Free
 }
 {
  bone1Name=dyn_rear_skirt_r_03
  bone2Name=dyn_rear_skirt_l_03
  swing1LimitDegrees=45
  swing1Motion=Free
  swing2LimitDegrees=45
  swing2Motion=Free
  twistLimitDegrees=45
  twistMotion=Free
 }
 {
  bone1Name=dyn_rear_skirt_l_03
  bone2Name=dyn_rear_skirt_r_03
  swing1LimitDegrees=45
  swing1Motion=Free
  swing2LimitDegrees=45
  swing2Motion=Free
  twistLimitDegrees=45
  twistMotion=Free
 }
 {
  bone1Name=dyn_rear_skirt_l_05
  bone2Name=dyn_rear_skirt_r_05
  swing1LimitDegrees=45
  swing1Motion=Free
  swing2LimitDegrees=45
  swing2Motion=Free
  twistLimitDegrees=45
  twistMotion=Free
 }
 {
  bone1Name=dyn_rear_skirt_r_05
  bone2Name=dyn_rear_skirt_l_05
  swing1LimitDegrees=45
  swing1Motion=Free
  swing2LimitDegrees=45
  swing2Motion=Free
  twistLimitDegrees=45
  twistMotion=Free
 }
]
