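# Animation Blueprint for the PureCereal medium male body cosmetic, adding secondary motion to its skirt and hand cloth.
# The pose copied from the attached parent mesh feeds a skirt Control Rig driven by movement, DBNO, emote and low-FPS states, then additive bone offsets adjust the skirt.
# The cached BlendedPoses runs through chained, spring-based AnimDynamics nodes simulating left and right hand cloth.
# The def_belt bone is scaled to zero before output.
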AnimBlueprint /BRCosmetics/Characters/Player/Male/Medium/Bodies/M_MED_PureCereal/Meshes/M_MED_PureCereal_AnimBP
target_skeleton=&/M_MED_PureCereal_Skeleton
graphs [
 AnimGraph {nodes [
   n1 = Anim Dynamics - Bone: dyn_hand_cloth_01_l!(
    ComponentPose=Local To Component(LocalPose=Use cached pose 'BlendedPoses'())
    Alpha=1
    SimSpaceSettings:Anim Phys Sim Space Settings Structure=(SimSpaceAngularAlpha=0,MaxAngularVelocity=10000,MaxAngularAcceleration=10000,ExternalAngularVelocity=(X=0,Y=0,Z=0))
    ExternalForce="0,0,0"
   ) -> (Pose) {
    angularDampingOverride=0.995
    angularSpringConstant=500
    bAngularSpring=true
    bOverrideAngularDamping=true
    bUseGravityOverride=true
    simulationSpace=RootRelative
    alphaBoolBlend {blendOption=Linear}
    boundBone {boneName=dyn_hand_cloth_01_l}
    componentAppliedLinearAccClamp=<3000 3000 3000>
    componentLinearAccScale=<0.4 0.4 0.4>
    componentLinearVelScale=<0.4 0.4 0.4>
    gravityOverride {z=-980}
    physicsBodyDefinitions [{
      boundBone {boneName=dyn_hand_cloth_01_l}
      boxExtents=<10 10 10>
      constraintSetup {angularLimitsMax=<5 20 60>, angularLimitsMin=<-5 -20 -60>, angularTarget=<0 0 0>}
      localJointOffset=<1 0 0>
     }]
    relativeSpaceBone {boneName=spine_05}
   }
   n2 = Anim Dynamics - Bone: dyn_hand_cloth_02_l!(
    ComponentPose=@n1
    Alpha=1
    SimSpaceSettings:Anim Phys Sim Space Settings Structure=(SimSpaceAngularAlpha=0,MaxAngularVelocity=10000,MaxAngularAcceleration=10000,ExternalAngularVelocity=(X=0,Y=0,Z=0))
    ExternalForce="0,0,0"
   ) -> (Pose) {
    angularDampingOverride=0.995
    angularSpringConstant=500
    bAngularSpring=true
    bOverrideAngularDamping=true
    bUseGravityOverride=true
    simulationSpace=RootRelative
    alphaBoolBlend {blendOption=Linear}
    boundBone {boneName=dyn_hand_cloth_02_l}
    componentAppliedLinearAccClamp=<3000 3000 3000>
    componentLinearAccScale=<0.4 0.4 0.4>
    componentLinearVelScale=<0.4 0.4 0.4>
    gravityOverride {z=-980}
    physicsBodyDefinitions [{
      boundBone {boneName=dyn_hand_cloth_02_l}
      boxExtents=<10 10 10>
      constraintSetup {angularLimitsMax=<5 20 60>, angularLimitsMin=<-5 -20 -60>, angularTarget=<0 0 0>}
      localJointOffset=<1.2 0 0>
     }]
    relativeSpaceBone {boneName=spine_05}
   }
   n3 = Anim Dynamics - Bone: dyn_hand_cloth_03_l!(
    ComponentPose=@n2
    Alpha=1
    SimSpaceSettings:Anim Phys Sim Space Settings Structure=(SimSpaceAngularAlpha=0,MaxAngularVelocity=10000,MaxAngularAcceleration=10000,ExternalAngularVelocity=(X=0,Y=0,Z=0))
    ExternalForce="0,0,0"
   ) -> (Pose) {
    angularDampingOverride=0.995
    angularSpringConstant=800
    bAngularSpring=true
    bOverrideAngularDamping=true
    bUseGravityOverride=true
    simulationSpace=RootRelative
    alphaBoolBlend {blendOption=Linear}
    boundBone {boneName=dyn_hand_cloth_03_l}
    componentAppliedLinearAccClamp=<3000 3000 3000>
    componentLinearAccScale=<0.4 0.4 0.4>
    componentLinearVelScale=<0.4 0.4 0.4>
    gravityOverride {z=-980}
    physicsBodyDefinitions [{
      boundBone {boneName=dyn_hand_cloth_03_l}
      boxExtents=<10 10 10>
      constraintSetup {angularLimitsMax=<5 20 45>, angularLimitsMin=<-5 -20 -45>, angularTarget=<0 0 0>}
      localJointOffset=<1.3 0 0>
     }]
    relativeSpaceBone {boneName=spine_05}
   }
   n4 = Anim Dynamics - Bone: dyn_hand_cloth_04_l!(
    ComponentPose=@n3
    Alpha=1
    SimSpaceSettings:Anim Phys Sim Space Settings Structure=(SimSpaceAngularAlpha=0,MaxAngularVelocity=10000,MaxAngularAcceleration=10000,ExternalAngularVelocity=(X=0,Y=0,Z=0))
    ExternalForce="0,0,0"
   ) -> (Pose) {
    angularDampingOverride=0.995
    angularSpringConstant=800
    bAngularSpring=true
    bOverrideAngularDamping=true
    bUseGravityOverride=true
    simulationSpace=RootRelative
    alphaBoolBlend {blendOption=Linear}
    boundBone {boneName=dyn_hand_cloth_04_l}
    componentAppliedLinearAccClamp=<3000 3000 3000>
    componentLinearAccScale=<0.4 0.4 0.4>
    componentLinearVelScale=<0.4 0.4 0.4>
    gravityOverride {z=-980}
    physicsBodyDefinitions [{
      boundBone {boneName=dyn_hand_cloth_04_l}
      boxExtents=<9 9 9>
      constraintSetup {angularLimitsMax=<5 20 45>, angularLimitsMin=<-5 -20 -45>, angularTarget=<0 0 0>}
      localJointOffset=<1.4 0 0>
     }]
    relativeSpaceBone {boneName=spine_05}
   }
   n5 = Anim Dynamics - Bone: dyn_hand_cloth_05_l!(
    ComponentPose=@n4
    Alpha=1
    SimSpaceSettings:Anim Phys Sim Space Settings Structure=(SimSpaceAngularAlpha=0,MaxAngularVelocity=10000,MaxAngularAcceleration=10000,ExternalAngularVelocity=(X=0,Y=0,Z=0))
    ExternalForce="0,0,0"
   ) -> (Pose) {
    angularDampingOverride=0.995
    angularSpringConstant=800
    bAngularSpring=true
    bOverrideAngularDamping=true
    bUseGravityOverride=true
    simulationSpace=RootRelative
    alphaBoolBlend {blendOption=Linear}
    boundBone {boneName=dyn_hand_cloth_05_l}
    componentAppliedLinearAccClamp=<3000 3000 3000>
    componentLinearAccScale=<0.4 0.4 0.4>
    componentLinearVelScale=<0.4 0.4 0.4>
    gravityOverride {z=-980}
    physicsBodyDefinitions [{
      boundBone {boneName=dyn_hand_cloth_05_l}
      boxExtents=<9 9 9>
      constraintSetup {angularLimitsMax=<5 20 30>, angularLimitsMin=<-5 -20 -30>, angularTarget=<0 0 0>}
      localJointOffset=<1.4 0 0>
     }]
    relativeSpaceBone {boneName=spine_05}
   }
   n6 = Anim Dynamics - Bone: dyn_hand_cloth_06_l!(
    ComponentPose=@n5
    Alpha=1
    SimSpaceSettings:Anim Phys Sim Space Settings Structure=(SimSpaceAngularAlpha=0,MaxAngularVelocity=10000,MaxAngularAcceleration=10000,ExternalAngularVelocity=(X=0,Y=0,Z=0))
    ExternalForce="0,0,0"
   ) -> (Pose) {
    angularDampingOverride=0.995
    angularSpringConstant=800
    bAngularSpring=true
    bOverrideAngularDamping=true
    bUseGravityOverride=true
    simulationSpace=RootRelative
    alphaBoolBlend {blendOption=Linear}
    boundBone {boneName=dyn_hand_cloth_06_l}
    componentAppliedLinearAccClamp=<3000 3000 3000>
    componentLinearAccScale=<0.4 0.4 0.4>
    componentLinearVelScale=<0.4 0.4 0.4>
    gravityOverride {z=-980}
    physicsBodyDefinitions [{
      boundBone {boneName=dyn_hand_cloth_06_l}
      boxExtents=<9 9 9>
      constraintSetup {angularLimitsMax=<5 20 45>, angularLimitsMin=<-5 -20 -45>, angularTarget=<0 0 0>}
      localJointOffset=<1.5 0 0>
     }]
    relativeSpaceBone {boneName=spine_05}
   }
   n7 = Anim Dynamics - Bone: dyn_hand_cloth_07_l!(
    ComponentPose=@n6
    Alpha=1
    SimSpaceSettings:Anim Phys Sim Space Settings Structure=(SimSpaceAngularAlpha=0,MaxAngularVelocity=10000,MaxAngularAcceleration=10000,ExternalAngularVelocity=(X=0,Y=0,Z=0))
    ExternalForce="0,0,0"
   ) -> (Pose) {
    angularDampingOverride=0.995
    angularSpringConstant=800
    bAngularSpring=true
    bOverrideAngularDamping=true
    bUseGravityOverride=true
    simulationSpace=RootRelative
    alphaBoolBlend {blendOption=Linear}
    boundBone {boneName=dyn_hand_cloth_07_l}
    componentAppliedLinearAccClamp=<3000 3000 3000>
    componentLinearAccScale=<0.4 0.4 0.4>
    componentLinearVelScale=<0.4 0.4 0.4>
    gravityOverride {z=-980}
    physicsBodyDefinitions [{
      boundBone {boneName=dyn_hand_cloth_07_l}
      boxExtents=<9 9 9>
      constraintSetup {angularLimitsMax=<5 20 30>, angularLimitsMin=<-5 -20 -30>, angularTarget=<0 0 0>}
      localJointOffset=<1.5 0 0>
     }]
    relativeSpaceBone {boneName=spine_05}
   }
   n8 = Anim Dynamics - Bone: dyn_hand_cloth_11_l!(
    ComponentPose=@n7
    Alpha=1
    SimSpaceSettings:Anim Phys Sim Space Settings Structure=(SimSpaceAngularAlpha=0,MaxAngularVelocity=10000,MaxAngularAcceleration=10000,ExternalAngularVelocity=(X=0,Y=0,Z=0))
    ExternalForce="0,0,0"
   ) -> (Pose) {
    angularDampingOverride=0.995
    angularSpringConstant=500
    bAngularSpring=true
    bOverrideAngularDamping=true
    bUseGravityOverride=true
    simulationSpace=RootRelative
    alphaBoolBlend {blendOption=Linear}
    boundBone {boneName=dyn_hand_cloth_11_l}
    componentAppliedLinearAccClamp=<3000 3000 3000>
    componentLinearAccScale=<0.4 0.4 0.4>
    componentLinearVelScale=<0.4 0.4 0.4>
    gravityOverride {z=-980}
    physicsBodyDefinitions [{
      boundBone {boneName=dyn_hand_cloth_11_l}
      boxExtents=<10 10 10>
      constraintSetup {angularLimitsMax=<5 20 60>, angularLimitsMin=<-5 -20 -60>, angularTarget=<0 0 0>}
      localJointOffset=<1 0 0>
     }]
    relativeSpaceBone {boneName=spine_05}
   }
   n9 = Anim Dynamics - Bone: dyn_hand_cloth_12_l!(
    ComponentPose=@n8
    Alpha=1
    SimSpaceSettings:Anim Phys Sim Space Settings Structure=(SimSpaceAngularAlpha=0,MaxAngularVelocity=10000,MaxAngularAcceleration=10000,ExternalAngularVelocity=(X=0,Y=0,Z=0))
    ExternalForce="0,0,0"
   ) -> (Pose) {
    angularDampingOverride=0.995
    angularSpringConstant=500
    bAngularSpring=true
    bOverrideAngularDamping=true
    bUseGravityOverride=true
    simulationSpace=RootRelative
    alphaBoolBlend {blendOption=Linear}
    boundBone {boneName=dyn_hand_cloth_12_l}
    componentAppliedLinearAccClamp=<3000 3000 3000>
    componentLinearAccScale=<0.4 0.4 0.4>
    componentLinearVelScale=<0.4 0.4 0.4>
    gravityOverride {z=-980}
    physicsBodyDefinitions [{
      boundBone {boneName=dyn_hand_cloth_12_l}
      boxExtents=<10 10 10>
      constraintSetup {angularLimitsMax=<5 20 60>, angularLimitsMin=<-5 -20 -60>, angularTarget=<0 0 0>}
      localJointOffset=<1.2 0 0>
     }]
    relativeSpaceBone {boneName=spine_05}
   }
   n10 = Anim Dynamics - Bone: dyn_hand_cloth_13_l!(
    ComponentPose=@n9
    Alpha=1
    SimSpaceSettings:Anim Phys Sim Space Settings Structure=(SimSpaceAngularAlpha=0,MaxAngularVelocity=10000,MaxAngularAcceleration=10000,ExternalAngularVelocity=(X=0,Y=0,Z=0))
    ExternalForce="0,0,0"
   ) -> (Pose) {
    angularDampingOverride=0.995
    angularSpringConstant=800
    bAngularSpring=true
    bOverrideAngularDamping=true
    bUseGravityOverride=true
    simulationSpace=RootRelative
    alphaBoolBlend {blendOption=Linear}
    boundBone {boneName=dyn_hand_cloth_13_l}
    componentAppliedLinearAccClamp=<3000 3000 3000>
    componentLinearAccScale=<0.4 0.4 0.4>
    componentLinearVelScale=<0.4 0.4 0.4>
    gravityOverride {z=-980}
    physicsBodyDefinitions [{
      boundBone {boneName=dyn_hand_cloth_13_l}
      boxExtents=<10 10 10>
      constraintSetup {angularLimitsMax=<5 20 30>, angularLimitsMin=<-5 -20 -30>, angularTarget=<0 0 0>}
      localJointOffset=<1.3 0 0>
     }]
    relativeSpaceBone {boneName=spine_05}
   }
   n11 = Anim Dynamics - Bone: dyn_hand_cloth_41_l!(
    ComponentPose=@n10
    Alpha=1
    SimSpaceSettings:Anim Phys Sim Space Settings Structure=(SimSpaceAngularAlpha=0,MaxAngularVelocity=10000,MaxAngularAcceleration=10000,ExternalAngularVelocity=(X=0,Y=0,Z=0))
    ExternalForce="0,0,0"
   ) -> (Pose) {
    angularDampingOverride=0.995
    angularSpringConstant=500
    bAngularSpring=true
    bOverrideAngularDamping=true
    bUseGravityOverride=true
    simulationSpace=RootRelative
    alphaBoolBlend {blendOption=Linear}
    boundBone {boneName=dyn_hand_cloth_41_l}
    componentAppliedLinearAccClamp=<3000 3000 3000>
    componentLinearAccScale=<0.4 0.4 0.4>
    componentLinearVelScale=<0.4 0.4 0.4>
    gravityOverride {z=-980}
    physicsBodyDefinitions [{
      boundBone {boneName=dyn_hand_cloth_41_l}
      boxExtents=<10 10 10>
      constraintSetup {angularLimitsMax=<5 20 60>, angularLimitsMin=<-5 -20 -60>, angularTarget=<0 0 0>}
      localJointOffset=<1 0 0>
     }]
    relativeSpaceBone {boneName=spine_05}
   }
   n12 = Anim Dynamics - Bone: dyn_hand_cloth_42_l!(
    ComponentPose=@n11
    Alpha=1
    SimSpaceSettings:Anim Phys Sim Space Settings Structure=(SimSpaceAngularAlpha=0,MaxAngularVelocity=10000,MaxAngularAcceleration=10000,ExternalAngularVelocity=(X=0,Y=0,Z=0))
    ExternalForce="0,0,0"
   ) -> (Pose) {
    angularDampingOverride=0.995
    angularSpringConstant=500
    bAngularSpring=true
    bOverrideAngularDamping=true
    bUseGravityOverride=true
    simulationSpace=RootRelative
    alphaBoolBlend {blendOption=Linear}
    boundBone {boneName=dyn_hand_cloth_42_l}
    componentAppliedLinearAccClamp=<3000 3000 3000>
    componentLinearAccScale=<0.4 0.4 0.4>
    componentLinearVelScale=<0.4 0.4 0.4>
    gravityOverride {z=-980}
    physicsBodyDefinitions [{
      boundBone {boneName=dyn_hand_cloth_42_l}
      boxExtents=<10 10 10>
      constraintSetup {angularLimitsMax=<5 20 60>, angularLimitsMin=<-5 -20 -60>, angularTarget=<0 0 0>}
      localJointOffset=<1.2 0 0>
     }]
    relativeSpaceBone {boneName=spine_05}
   }
   n13 = Anim Dynamics - Bone: dyn_hand_cloth_43_l!(
    ComponentPose=@n12
    Alpha=1
    SimSpaceSettings:Anim Phys Sim Space Settings Structure=(SimSpaceAngularAlpha=0,MaxAngularVelocity=10000,MaxAngularAcceleration=10000,ExternalAngularVelocity=(X=0,Y=0,Z=0))
    ExternalForce="0,0,0"
   ) -> (Pose) {
    angularDampingOverride=0.995
    angularSpringConstant=800
    bAngularSpring=true
    bOverrideAngularDamping=true
    bUseGravityOverride=true
    simulationSpace=RootRelative
    alphaBoolBlend {blendOption=Linear}
    boundBone {boneName=dyn_hand_cloth_43_l}
    componentAppliedLinearAccClamp=<3000 3000 3000>
    componentLinearAccScale=<0.4 0.4 0.4>
    componentLinearVelScale=<0.4 0.4 0.4>
    gravityOverride {z=-980}
    physicsBodyDefinitions [{
      boundBone {boneName=dyn_hand_cloth_43_l}
      boxExtents=<10 10 10>
      constraintSetup {angularLimitsMax=<5 20 45>, angularLimitsMin=<-5 -20 -45>, angularTarget=<0 0 0>}
      localJointOffset=<1.3 0 0>
     }]
    relativeSpaceBone {boneName=spine_05}
   }
   n14 = Anim Dynamics - Bone: dyn_hand_cloth_44_l!(
    ComponentPose=@n13
    Alpha=1
    SimSpaceSettings:Anim Phys Sim Space Settings Structure=(SimSpaceAngularAlpha=0,MaxAngularVelocity=10000,MaxAngularAcceleration=10000,ExternalAngularVelocity=(X=0,Y=0,Z=0))
    ExternalForce="0,0,0"
   ) -> (Pose) {
    angularDampingOverride=0.995
    angularSpringConstant=800
    bAngularSpring=true
    bOverrideAngularDamping=true
    bUseGravityOverride=true
    simulationSpace=RootRelative
    alphaBoolBlend {blendOption=Linear}
    boundBone {boneName=dyn_hand_cloth_44_l}
    componentAppliedLinearAccClamp=<3000 3000 3000>
    componentLinearAccScale=<0.4 0.4 0.4>
    componentLinearVelScale=<0.4 0.4 0.4>
    gravityOverride {z=-980}
    physicsBodyDefinitions [{
      boundBone {boneName=dyn_hand_cloth_44_l}
      boxExtents=<9 9 9>
      constraintSetup {angularLimitsMax=<5 20 45>, angularLimitsMin=<-5 -20 -45>, angularTarget=<0 0 0>}
      localJointOffset=<1.4 0 0>
     }]
    relativeSpaceBone {boneName=spine_05}
   }
   n15 = Anim Dynamics - Bone: dyn_hand_cloth_45_l!(
    ComponentPose=@n14
    Alpha=1
    SimSpaceSettings:Anim Phys Sim Space Settings Structure=(SimSpaceAngularAlpha=0,MaxAngularVelocity=10000,MaxAngularAcceleration=10000,ExternalAngularVelocity=(X=0,Y=0,Z=0))
    ExternalForce="0,0,0"
   ) -> (Pose) {
    angularDampingOverride=0.995
    angularSpringConstant=800
    bAngularSpring=true
    bOverrideAngularDamping=true
    bUseGravityOverride=true
    simulationSpace=RootRelative
    alphaBoolBlend {blendOption=Linear}
    boundBone {boneName=dyn_hand_cloth_45_l}
    componentAppliedLinearAccClamp=<3000 3000 3000>
    componentLinearAccScale=<0.4 0.4 0.4>
    componentLinearVelScale=<0.4 0.4 0.4>
    gravityOverride {z=-980}
    physicsBodyDefinitions [{
      boundBone {boneName=dyn_hand_cloth_45_l}
      boxExtents=<9 9 9>
      constraintSetup {angularLimitsMax=<5 20 45>, angularLimitsMin=<-5 -20 -45>, angularTarget=<0 0 0>}
      localJointOffset=<1.5 0 0>
     }]
    relativeSpaceBone {boneName=spine_05}
   }
   n16 = Anim Dynamics - Bone: dyn_hand_cloth_21_l!(
    ComponentPose=@n15
    Alpha=1
    SimSpaceSettings:Anim Phys Sim Space Settings Structure=(SimSpaceAngularAlpha=0,MaxAngularVelocity=10000,MaxAngularAcceleration=10000,ExternalAngularVelocity=(X=0,Y=0,Z=0))
    ExternalForce="0,0,0"
   ) -> (Pose) {
    angularDampingOverride=0.995
    angularSpringConstant=500
    bAngularSpring=true
    bOverrideAngularDamping=true
    bUseGravityOverride=true
    simulationSpace=RootRelative
    alphaBoolBlend {blendOption=Linear}
    boundBone {boneName=dyn_hand_cloth_21_l}
    componentAppliedLinearAccClamp=<3000 3000 3000>
    componentLinearAccScale=<0.4 0.4 0.4>
    componentLinearVelScale=<0.4 0.4 0.4>
    gravityOverride {z=-980}
    physicsBodyDefinitions [{
      boundBone {boneName=dyn_hand_cloth_21_l}
      boxExtents=<10 10 10>
      constraintSetup {angularLimitsMax=<5 20 60>, angularLimitsMin=<-5 -20 -60>, angularTarget=<0 0 0>}
      localJointOffset=<1 0 0>
     }]
    relativeSpaceBone {boneName=spine_05}
   }
   n17 = Anim Dynamics - Bone: dyn_hand_cloth_22_l!(
    ComponentPose=@n16
    Alpha=1
    SimSpaceSettings:Anim Phys Sim Space Settings Structure=(SimSpaceAngularAlpha=0,MaxAngularVelocity=10000,MaxAngularAcceleration=10000,ExternalAngularVelocity=(X=0,Y=0,Z=0))
    ExternalForce="0,0,0"
   ) -> (Pose) {
    angularDampingOverride=0.995
    angularSpringConstant=500
    bAngularSpring=true
    bOverrideAngularDamping=true
    bUseGravityOverride=true
    simulationSpace=RootRelative
    alphaBoolBlend {blendOption=Linear}
    boundBone {boneName=dyn_hand_cloth_22_l}
    componentAppliedLinearAccClamp=<3000 3000 3000>
    componentLinearAccScale=<0.4 0.4 0.4>
    componentLinearVelScale=<0.4 0.4 0.4>
    gravityOverride {z=-980}
    physicsBodyDefinitions [{
      boundBone {boneName=dyn_hand_cloth_22_l}
      boxExtents=<10 10 10>
      constraintSetup {angularLimitsMax=<5 20 60>, angularLimitsMin=<-5 -20 -60>, angularTarget=<0 0 0>}
      localJointOffset=<1.2 0 0>
     }]
    relativeSpaceBone {boneName=spine_05}
   }
   n18 = Anim Dynamics - Bone: dyn_hand_cloth_23_l!(
    ComponentPose=@n17
    Alpha=1
    SimSpaceSettings:Anim Phys Sim Space Settings Structure=(SimSpaceAngularAlpha=0,MaxAngularVelocity=10000,MaxAngularAcceleration=10000,ExternalAngularVelocity=(X=0,Y=0,Z=0))
    ExternalForce="0,0,0"
   ) -> (Pose) {
    angularDampingOverride=0.995
    angularSpringConstant=800
    bAngularSpring=true
    bOverrideAngularDamping=true
    bUseGravityOverride=true
    simulationSpace=RootRelative
    alphaBoolBlend {blendOption=Linear}
    boundBone {boneName=dyn_hand_cloth_23_l}
    componentAppliedLinearAccClamp=<3000 3000 3000>
    componentLinearAccScale=<0.4 0.4 0.4>
    componentLinearVelScale=<0.4 0.4 0.4>
    gravityOverride {z=-980}
    physicsBodyDefinitions [{
      boundBone {boneName=dyn_hand_cloth_23_l}
      boxExtents=<10 10 10>
      constraintSetup {angularLimitsMax=<5 20 45>, angularLimitsMin=<-5 -20 -45>, angularTarget=<0 0 0>}
      localJointOffset=<1.3 0 0>
     }]
    relativeSpaceBone {boneName=spine_05}
   }
   n19 = Anim Dynamics - Bone: dyn_hand_cloth_24_l!(
    ComponentPose=@n18
    Alpha=1
    SimSpaceSettings:Anim Phys Sim Space Settings Structure=(SimSpaceAngularAlpha=0,MaxAngularVelocity=10000,MaxAngularAcceleration=10000,ExternalAngularVelocity=(X=0,Y=0,Z=0))
    ExternalForce="0,0,0"
   ) -> (Pose) {
    angularDampingOverride=0.995
    angularSpringConstant=800
    bAngularSpring=true
    bOverrideAngularDamping=true
    bUseGravityOverride=true
    simulationSpace=RootRelative
    alphaBoolBlend {blendOption=Linear}
    boundBone {boneName=dyn_hand_cloth_24_l}
    componentAppliedLinearAccClamp=<3000 3000 3000>
    componentLinearAccScale=<0.4 0.4 0.4>
    componentLinearVelScale=<0.4 0.4 0.4>
    gravityOverride {z=-980}
    physicsBodyDefinitions [{
      boundBone {boneName=dyn_hand_cloth_24_l}
      boxExtents=<9 9 9>
      constraintSetup {angularLimitsMax=<5 20 45>, angularLimitsMin=<-5 -20 -45>, angularTarget=<0 0 0>}
      localJointOffset=<1.4 0 0>
     }]
    relativeSpaceBone {boneName=spine_05}
   }
   n20 = Anim Dynamics - Bone: dyn_hand_cloth_31_l!(
    ComponentPose=@n19
    Alpha=1
    SimSpaceSettings:Anim Phys Sim Space Settings Structure=(SimSpaceAngularAlpha=0,MaxAngularVelocity=10000,MaxAngularAcceleration=10000,ExternalAngularVelocity=(X=0,Y=0,Z=0))
    ExternalForce="0,0,0"
   ) -> (Pose) {
    angularDampingOverride=0.995
    angularSpringConstant=500
    bAngularSpring=true
    bOverrideAngularDamping=true
    bUseGravityOverride=true
    simulationSpace=RootRelative
    alphaBoolBlend {blendOption=Linear}
    boundBone {boneName=dyn_hand_cloth_31_l}
    componentAppliedLinearAccClamp=<3000 3000 3000>
    componentLinearAccScale=<0.4 0.4 0.4>
    componentLinearVelScale=<0.4 0.4 0.4>
    gravityOverride {z=-980}
    physicsBodyDefinitions [{
      boundBone {boneName=dyn_hand_cloth_31_l}
      boxExtents=<10 10 10>
      constraintSetup {angularLimitsMax=<5 20 60>, angularLimitsMin=<-5 -20 -60>, angularTarget=<0 0 0>}
      localJointOffset=<1 0 0>
     }]
    relativeSpaceBone {boneName=spine_05}
   }
   n21 = Anim Dynamics - Bone: dyn_hand_cloth_32_l!(
    ComponentPose=@n20
    Alpha=1
    SimSpaceSettings:Anim Phys Sim Space Settings Structure=(SimSpaceAngularAlpha=0,MaxAngularVelocity=10000,MaxAngularAcceleration=10000,ExternalAngularVelocity=(X=0,Y=0,Z=0))
    ExternalForce="0,0,0"
   ) -> (Pose) {
    angularDampingOverride=0.995
    angularSpringConstant=500
    bAngularSpring=true
    bOverrideAngularDamping=true
    bUseGravityOverride=true
    simulationSpace=RootRelative
    alphaBoolBlend {blendOption=Linear}
    boundBone {boneName=dyn_hand_cloth_32_l}
    componentAppliedLinearAccClamp=<3000 3000 3000>
    componentLinearAccScale=<0.4 0.4 0.4>
    componentLinearVelScale=<0.4 0.4 0.4>
    gravityOverride {z=-980}
    physicsBodyDefinitions [{
      boundBone {boneName=dyn_hand_cloth_32_l}
      boxExtents=<10 10 10>
      constraintSetup {angularLimitsMax=<5 20 60>, angularLimitsMin=<-5 -20 -60>, angularTarget=<0 0 0>}
      localJointOffset=<1.2 0 0>
     }]
    relativeSpaceBone {boneName=spine_05}
   }
   n22 = Anim Dynamics - Bone: dyn_hand_cloth_01_r!(
    ComponentPose=@n21
    Alpha=1
    SimSpaceSettings:Anim Phys Sim Space Settings Structure=(SimSpaceAngularAlpha=0,MaxAngularVelocity=10000,MaxAngularAcceleration=10000,ExternalAngularVelocity=(X=0,Y=0,Z=0))
    ExternalForce="0,0,0"
   ) -> (Pose) {
    angularDampingOverride=0.995
    angularSpringConstant=500
    bAngularSpring=true
    bOverrideAngularDamping=true
    bUseGravityOverride=true
    simulationSpace=RootRelative
    alphaBoolBlend {blendOption=Linear}
    boundBone {boneName=dyn_hand_cloth_01_r}
    componentAppliedLinearAccClamp=<3000 3000 3000>
    componentLinearAccScale=<0.4 0.4 0.4>
    componentLinearVelScale=<0.4 0.4 0.4>
    gravityOverride {z=-980}
    physicsBodyDefinitions [{
      boundBone {boneName=dyn_hand_cloth_01_r}
      boxExtents=<10 10 10>
      constraintSetup {angularLimitsMax=<5 20 75>, angularLimitsMin=<-5 -20 -75>, angularTarget=<0 0 0>}
      localJointOffset=<-1 0 0>
     }]
    relativeSpaceBone {boneName=spine_05}
   }
   n23 = Anim Dynamics - Bone: dyn_hand_cloth_02_r!(
    ComponentPose=@n22
    Alpha=1
    SimSpaceSettings:Anim Phys Sim Space Settings Structure=(SimSpaceAngularAlpha=0,MaxAngularVelocity=10000,MaxAngularAcceleration=10000,ExternalAngularVelocity=(X=0,Y=0,Z=0))
    ExternalForce="0,0,0"
   ) -> (Pose) {
    angularDampingOverride=0.995
    angularSpringConstant=500
    bAngularSpring=true
    bOverrideAngularDamping=true
    bUseGravityOverride=true
    simulationSpace=RootRelative
    alphaBoolBlend {blendOption=Linear}
    boundBone {boneName=dyn_hand_cloth_02_r}
    componentAppliedLinearAccClamp=<3000 3000 3000>
    componentLinearAccScale=<0.4 0.4 0.4>
    componentLinearVelScale=<0.4 0.4 0.4>
    gravityOverride {z=-980}
    physicsBodyDefinitions [{
      boundBone {boneName=dyn_hand_cloth_02_r}
      boxExtents=<10 10 10>
      constraintSetup {angularLimitsMax=<5 20 60>, angularLimitsMin=<-5 -20 -60>, angularTarget=<0 0 0>}
      localJointOffset=<-1.2 0 0>
     }]
    relativeSpaceBone {boneName=spine_05}
   }
   n24 = Anim Dynamics - Bone: dyn_hand_cloth_03_r!(
    ComponentPose=@n23
    Alpha=1
    SimSpaceSettings:Anim Phys Sim Space Settings Structure=(SimSpaceAngularAlpha=0,MaxAngularVelocity=10000,MaxAngularAcceleration=10000,ExternalAngularVelocity=(X=0,Y=0,Z=0))
    ExternalForce="0,0,0"
   ) -> (Pose) {
    angularDampingOverride=0.995
    angularSpringConstant=800
    bAngularSpring=true
    bOverrideAngularDamping=true
    bUseGravityOverride=true
    simulationSpace=RootRelative
    alphaBoolBlend {blendOption=Linear}
    boundBone {boneName=dyn_hand_cloth_03_r}
    componentAppliedLinearAccClamp=<3000 3000 3000>
    componentLinearAccScale=<0.4 0.4 0.4>
    componentLinearVelScale=<0.4 0.4 0.4>
    gravityOverride {z=-980}
    physicsBodyDefinitions [{
      boundBone {boneName=dyn_hand_cloth_03_r}
      boxExtents=<10 10 10>
      constraintSetup {angularLimitsMax=<5 20 45>, angularLimitsMin=<-5 -20 -45>, angularTarget=<0 0 0>}
      localJointOffset=<-1.3 0 0>
     }]
    relativeSpaceBone {boneName=spine_05}
   }
   n25 = Anim Dynamics - Bone: dyn_hand_cloth_04_r!(
    ComponentPose=@n24
    Alpha=1
    SimSpaceSettings:Anim Phys Sim Space Settings Structure=(SimSpaceAngularAlpha=0,MaxAngularVelocity=10000,MaxAngularAcceleration=10000,ExternalAngularVelocity=(X=0,Y=0,Z=0))
    ExternalForce="0,0,0"
   ) -> (Pose) {
    angularDampingOverride=0.995
    angularSpringConstant=800
    bAngularSpring=true
    bOverrideAngularDamping=true
    bUseGravityOverride=true
    simulationSpace=RootRelative
    alphaBoolBlend {blendOption=Linear}
    boundBone {boneName=dyn_hand_cloth_04_r}
    componentAppliedLinearAccClamp=<3000 3000 3000>
    componentLinearAccScale=<0.4 0.4 0.4>
    componentLinearVelScale=<0.4 0.4 0.4>
    gravityOverride {z=-980}
    physicsBodyDefinitions [{
      boundBone {boneName=dyn_hand_cloth_04_r}
      boxExtents=<9 9 9>
      constraintSetup {angularLimitsMax=<5 20 45>, angularLimitsMin=<-5 -20 -45>, angularTarget=<0 0 0>}
      localJointOffset=<-1.4 0 0>
     }]
    relativeSpaceBone {boneName=spine_05}
   }
   n26 = Anim Dynamics - Bone: dyn_hand_cloth_05_r!(
    ComponentPose=@n25
    Alpha=1
    SimSpaceSettings:Anim Phys Sim Space Settings Structure=(SimSpaceAngularAlpha=0,MaxAngularVelocity=10000,MaxAngularAcceleration=10000,ExternalAngularVelocity=(X=0,Y=0,Z=0))
    ExternalForce="0,0,0"
   ) -> (Pose) {
    angularDampingOverride=0.995
    angularSpringConstant=800
    bAngularSpring=true
    bOverrideAngularDamping=true
    bUseGravityOverride=true
    simulationSpace=RootRelative
    alphaBoolBlend {blendOption=Linear}
    boundBone {boneName=dyn_hand_cloth_05_r}
    componentAppliedLinearAccClamp=<3000 3000 3000>
    componentLinearAccScale=<0.4 0.4 0.4>
    componentLinearVelScale=<0.4 0.4 0.4>
    gravityOverride {z=-980}
    physicsBodyDefinitions [{
      boundBone {boneName=dyn_hand_cloth_05_r}
      boxExtents=<9 9 9>
      constraintSetup {angularLimitsMax=<5 20 45>, angularLimitsMin=<-5 -20 -45>, angularTarget=<0 0 0>}
      localJointOffset=<-1.4 0 0>
     }]
    relativeSpaceBone {boneName=spine_05}
   }
   n27 = Anim Dynamics - Bone: dyn_hand_cloth_06_r!(
    ComponentPose=@n26
    Alpha=1
    SimSpaceSettings:Anim Phys Sim Space Settings Structure=(SimSpaceAngularAlpha=0,MaxAngularVelocity=10000,MaxAngularAcceleration=10000,ExternalAngularVelocity=(X=0,Y=0,Z=0))
    ExternalForce="0,0,0"
   ) -> (Pose) {
    angularDampingOverride=0.995
    angularSpringConstant=800
    bAngularSpring=true
    bOverrideAngularDamping=true
    bUseGravityOverride=true
    simulationSpace=RootRelative
    alphaBoolBlend {blendOption=Linear}
    boundBone {boneName=dyn_hand_cloth_06_r}
    componentAppliedLinearAccClamp=<3000 3000 3000>
    componentLinearAccScale=<0.4 0.4 0.4>
    componentLinearVelScale=<0.4 0.4 0.4>
    gravityOverride {z=-980}
    physicsBodyDefinitions [{
      boundBone {boneName=dyn_hand_cloth_06_r}
      boxExtents=<10 10 10>
      constraintSetup {angularLimitsMax=<5 20 45>, angularLimitsMin=<-5 -20 -45>, angularTarget=<0 0 0>}
      localJointOffset=<-1.3 0 0>
     }]
    relativeSpaceBone {boneName=spine_05}
   }
   n28 = Anim Dynamics - Bone: dyn_hand_cloth_11_r!(
    ComponentPose=@n27
    Alpha=1
    SimSpaceSettings:Anim Phys Sim Space Settings Structure=(SimSpaceAngularAlpha=0,MaxAngularVelocity=10000,MaxAngularAcceleration=10000,ExternalAngularVelocity=(X=0,Y=0,Z=0))
    ExternalForce="0,0,0"
   ) -> (Pose) {
    angularDampingOverride=0.995
    angularSpringConstant=500
    bAngularSpring=true
    bOverrideAngularDamping=true
    bUseGravityOverride=true
    simulationSpace=RootRelative
    alphaBoolBlend {blendOption=Linear}
    boundBone {boneName=dyn_hand_cloth_11_r}
    componentAppliedLinearAccClamp=<3000 3000 3000>
    componentLinearAccScale=<0.4 0.4 0.4>
    componentLinearVelScale=<0.4 0.4 0.4>
    gravityOverride {z=-980}
    physicsBodyDefinitions [{
      boundBone {boneName=dyn_hand_cloth_11_r}
      boxExtents=<10 10 10>
      constraintSetup {angularLimitsMax=<5 20 75>, angularLimitsMin=<-5 -20 -75>, angularTarget=<0 0 0>}
      localJointOffset=<-1 0 0>
     }]
    relativeSpaceBone {boneName=spine_05}
   }
   n29 = Anim Dynamics - Bone: dyn_hand_cloth_12_r!(
    ComponentPose=@n28
    Alpha=1
    SimSpaceSettings:Anim Phys Sim Space Settings Structure=(SimSpaceAngularAlpha=0,MaxAngularVelocity=10000,MaxAngularAcceleration=10000,ExternalAngularVelocity=(X=0,Y=0,Z=0))
    ExternalForce="0,0,0"
   ) -> (Pose) {
    angularDampingOverride=0.995
    angularSpringConstant=500
    bAngularSpring=true
    bOverrideAngularDamping=true
    bUseGravityOverride=true
    simulationSpace=RootRelative
    alphaBoolBlend {blendOption=Linear}
    boundBone {boneName=dyn_hand_cloth_12_r}
    componentAppliedLinearAccClamp=<3000 3000 3000>
    componentLinearAccScale=<0.4 0.4 0.4>
    componentLinearVelScale=<0.4 0.4 0.4>
    gravityOverride {z=-980}
    physicsBodyDefinitions [{
      boundBone {boneName=dyn_hand_cloth_12_r}
      boxExtents=<10 10 10>
      constraintSetup {angularLimitsMax=<5 20 60>, angularLimitsMin=<-5 -20 -60>, angularTarget=<0 0 0>}
      localJointOffset=<-1.2 0 0>
     }]
    relativeSpaceBone {boneName=spine_05}
   }
   n30 = Anim Dynamics - Bone: dyn_hand_cloth_41_r!(
    ComponentPose=@n29
    Alpha=1
    SimSpaceSettings:Anim Phys Sim Space Settings Structure=(SimSpaceAngularAlpha=0,MaxAngularVelocity=10000,MaxAngularAcceleration=10000,ExternalAngularVelocity=(X=0,Y=0,Z=0))
    ExternalForce="0,0,0"
   ) -> (Pose) {
    angularDampingOverride=0.995
    angularSpringConstant=500
    bAngularSpring=true
    bOverrideAngularDamping=true
    bUseGravityOverride=true
    simulationSpace=RootRelative
    alphaBoolBlend {blendOption=Linear}
    boundBone {boneName=dyn_hand_cloth_41_r}
    componentAppliedLinearAccClamp=<3000 3000 3000>
    componentLinearAccScale=<0.4 0.4 0.4>
    componentLinearVelScale=<0.4 0.4 0.4>
    gravityOverride {z=-980}
    physicsBodyDefinitions [{
      boundBone {boneName=dyn_hand_cloth_41_r}
      boxExtents=<10 10 10>
      constraintSetup {angularLimitsMax=<5 20 75>, angularLimitsMin=<-5 -20 -75>, angularTarget=<0 0 0>}
      localJointOffset=<-1 0 0>
     }]
    relativeSpaceBone {boneName=spine_05}
   }
   n31 = Anim Dynamics - Bone: dyn_hand_cloth_42_r!(
    ComponentPose=@n30
    Alpha=1
    SimSpaceSettings:Anim Phys Sim Space Settings Structure=(SimSpaceAngularAlpha=0,MaxAngularVelocity=10000,MaxAngularAcceleration=10000,ExternalAngularVelocity=(X=0,Y=0,Z=0))
    ExternalForce="0,0,0"
   ) -> (Pose) {
    angularDampingOverride=0.995
    angularSpringConstant=500
    bAngularSpring=true
    bOverrideAngularDamping=true
    bUseGravityOverride=true
    simulationSpace=RootRelative
    alphaBoolBlend {blendOption=Linear}
    boundBone {boneName=dyn_hand_cloth_42_r}
    componentAppliedLinearAccClamp=<3000 3000 3000>
    componentLinearAccScale=<0.4 0.4 0.4>
    componentLinearVelScale=<0.4 0.4 0.4>
    gravityOverride {z=-980}
    physicsBodyDefinitions [{
      boundBone {boneName=dyn_hand_cloth_42_r}
      boxExtents=<10 10 10>
      constraintSetup {angularLimitsMax=<5 20 60>, angularLimitsMin=<-5 -20 -60>, angularTarget=<0 0 0>}
      localJointOffset=<-1.2 0 0>
     }]
    relativeSpaceBone {boneName=spine_05}
   }
   n32 = Anim Dynamics - Bone: dyn_hand_cloth_43_r!(
    ComponentPose=@n31
    Alpha=1
    SimSpaceSettings:Anim Phys Sim Space Settings Structure=(SimSpaceAngularAlpha=0,MaxAngularVelocity=10000,MaxAngularAcceleration=10000,ExternalAngularVelocity=(X=0,Y=0,Z=0))
    ExternalForce="0,0,0"
   ) -> (Pose) {
    angularDampingOverride=0.995
    angularSpringConstant=800
    bAngularSpring=true
    bOverrideAngularDamping=true
    bUseGravityOverride=true
    simulationSpace=RootRelative
    alphaBoolBlend {blendOption=Linear}
    boundBone {boneName=dyn_hand_cloth_43_r}
    componentAppliedLinearAccClamp=<3000 3000 3000>
    componentLinearAccScale=<0.4 0.4 0.4>
    componentLinearVelScale=<0.4 0.4 0.4>
    gravityOverride {z=-980}
    physicsBodyDefinitions [{
      boundBone {boneName=dyn_hand_cloth_43_r}
      boxExtents=<10 10 10>
      constraintSetup {angularLimitsMax=<5 20 45>, angularLimitsMin=<-5 -20 -45>, angularTarget=<0 0 0>}
      localJointOffset=<-1.3 0 0>
     }]
    relativeSpaceBone {boneName=spine_05}
   }
   n33 = Anim Dynamics - Bone: dyn_hand_cloth_21_r!(
    ComponentPose=@n32
    Alpha=1
    SimSpaceSettings:Anim Phys Sim Space Settings Structure=(SimSpaceAngularAlpha=0,MaxAngularVelocity=10000,MaxAngularAcceleration=10000,ExternalAngularVelocity=(X=0,Y=0,Z=0))
    ExternalForce="0,0,0"
   ) -> (Pose) {
    angularDampingOverride=0.995
    angularSpringConstant=500
    bAngularSpring=true
    bOverrideAngularDamping=true
    bUseGravityOverride=true
    simulationSpace=RootRelative
    alphaBoolBlend {blendOption=Linear}
    boundBone {boneName=dyn_hand_cloth_21_r}
    componentAppliedLinearAccClamp=<3000 3000 3000>
    componentLinearAccScale=<0.4 0.4 0.4>
    componentLinearVelScale=<0.4 0.4 0.4>
    gravityOverride {z=-980}
    physicsBodyDefinitions [{
      boundBone {boneName=dyn_hand_cloth_21_r}
      boxExtents=<10 10 10>
      constraintSetup {angularLimitsMax=<5 20 75>, angularLimitsMin=<-5 -20 -75>, angularTarget=<0 0 0>}
      localJointOffset=<-1 0 0>
     }]
    relativeSpaceBone {boneName=spine_05}
   }
   n34 = Anim Dynamics - Bone: dyn_hand_cloth_22_r!(
    ComponentPose=@n33
    Alpha=1
    SimSpaceSettings:Anim Phys Sim Space Settings Structure=(SimSpaceAngularAlpha=0,MaxAngularVelocity=10000,MaxAngularAcceleration=10000,ExternalAngularVelocity=(X=0,Y=0,Z=0))
    ExternalForce="0,0,0"
   ) -> (Pose) {
    angularDampingOverride=0.995
    angularSpringConstant=500
    bAngularSpring=true
    bOverrideAngularDamping=true
    bUseGravityOverride=true
    simulationSpace=RootRelative
    alphaBoolBlend {blendOption=Linear}
    boundBone {boneName=dyn_hand_cloth_22_r}
    componentAppliedLinearAccClamp=<3000 3000 3000>
    componentLinearAccScale=<0.4 0.4 0.4>
    componentLinearVelScale=<0.4 0.4 0.4>
    gravityOverride {z=-980}
    physicsBodyDefinitions [{
      boundBone {boneName=dyn_hand_cloth_22_r}
      boxExtents=<10 10 10>
      constraintSetup {angularLimitsMax=<5 20 60>, angularLimitsMin=<-5 -20 -60>, angularTarget=<0 0 0>}
      localJointOffset=<-1.2 0 0>
     }]
    relativeSpaceBone {boneName=spine_05}
   }
   n35 = Anim Dynamics - Bone: dyn_hand_cloth_23_r!(
    ComponentPose=@n34
    Alpha=1
    SimSpaceSettings:Anim Phys Sim Space Settings Structure=(SimSpaceAngularAlpha=0,MaxAngularVelocity=10000,MaxAngularAcceleration=10000,ExternalAngularVelocity=(X=0,Y=0,Z=0))
    ExternalForce="0,0,0"
   ) -> (Pose) {
    angularDampingOverride=0.995
    angularSpringConstant=800
    bAngularSpring=true
    bOverrideAngularDamping=true
    bUseGravityOverride=true
    simulationSpace=RootRelative
    alphaBoolBlend {blendOption=Linear}
    boundBone {boneName=dyn_hand_cloth_23_r}
    componentAppliedLinearAccClamp=<3000 3000 3000>
    componentLinearAccScale=<0.4 0.4 0.4>
    componentLinearVelScale=<0.4 0.4 0.4>
    gravityOverride {z=-980}
    physicsBodyDefinitions [{
      boundBone {boneName=dyn_hand_cloth_23_r}
      boxExtents=<10 10 10>
      constraintSetup {angularLimitsMax=<5 20 45>, angularLimitsMin=<-5 -20 -45>, angularTarget=<0 0 0>}
      localJointOffset=<-1.3 0 0>
     }]
    relativeSpaceBone {boneName=spine_05}
   }
   n36 = "Transform (Modify) Bone - Bone: def_belt"(ComponentPose=@n35, Alpha=1, Translation="0,0,0", Rotation="0,0,0", Scale="0,0,0") {scaleMode=BMM_Replace, scaleSpace=BCS_BoneSpace, alphaBoolBlend {blendOption=Linear}, boneToModify {boneName=def_belt}, scale=<0 0 0>}
   Output Pose(Result=Component To Local(ComponentPose=@n36)) {layerGroup=None}
   n37 = Control Rig!(
    Source=Copy Pose From Mesh() {bUseAttachedParent=true}
    Alpha
    IsKneeSliding:bool=Get bIsKneeSliding()
    IsBaseballSliding:bool=Get bIsBaseballSliding()
    IsCrouching:bool=Get bIsCrouching()
    MoveGroundCollision:bool=Get MoveGroundCollision() -> (MoveGroundCollision:bool)
    isDBNO:bool=Get bIsDBNO()
    isDBNOCarried:bool=Get bIsDBNOCarried()
    isPlayingEmote:bool=Get bIsPlayingEmote()
    isFrontEndNotEmote:double=Get FrontEndIdleModify() -> (FrontEndIdleModify:double)
    isFPSUnder30:bool=Get isFPSLowerThan30() -> (isFPSLowerThan30:bool)
   ) -> (Pose) {
    alphaBoolBlend {blendOption=Linear}
    controlRigAssetReference {blueprintRigClass=&/M_MED_PureCerealSkirt_CtrlRig.M_MED_PureCerealSkirt_CtrlRig_C}
   }
   Save cached pose 'PostDeformation'(Pose=@n37) {cacheName=PostDeformation}
   n38 = "Transform (Modify) Bone - Bone: dyn_front_skirt_l_MSTR"(
    ComponentPose=Local To Component(LocalPose=Use cached pose 'PostDeformation'())
    Alpha=1
    Translation="0,-1,0"
    Rotation="0,-2.5,0"
    Scale="1,1,1"
   ) {
    rotationMode=BMM_Additive
    rotationSpace=BCS_BoneSpace
    translationMode=BMM_Additive
    translationSpace=BCS_BoneSpace
    alphaBoolBlend {blendOption=Linear}
    boneToModify {boneName=dyn_front_skirt_l_MSTR}
    rotation {yaw=-2.5}
    translation {y=-1}
   }
   n39 = "Transform (Modify) Bone - Bone: dyn_front_skirt_corner_l_MSTR"(ComponentPose=@n38, Alpha=1, Translation="0,-1,0", Rotation="0,-2.5,0", Scale="1,1,1") {
    rotationMode=BMM_Additive
    rotationSpace=BCS_BoneSpace
    translationMode=BMM_Additive
    translationSpace=BCS_BoneSpace
    alphaBoolBlend {blendOption=Linear}
    boneToModify {boneName=dyn_front_skirt_corner_l_MSTR}
    rotation {yaw=-2.5}
    translation {y=-1}
   }
   n40 = "Transform (Modify) Bone - Bone: dyn_side_skirt_l_MSTR"(ComponentPose=@n39, Alpha=1, Translation="0,-1,0", Rotation="0,-2.5,0", Scale="1,1,1") {
    rotationMode=BMM_Additive
    rotationSpace=BCS_BoneSpace
    translationMode=BMM_Additive
    translationSpace=BCS_BoneSpace
    alphaBoolBlend {blendOption=Linear}
    boneToModify {boneName=dyn_side_skirt_l_MSTR}
    rotation {yaw=-2.5}
    translation {y=-1}
   }
   n41 = "Transform (Modify) Bone - Bone: dyn_rear_skirt_corner_l_MSTR"(ComponentPose=@n40, Alpha=0.5, Translation="0,-1,0", Rotation="0,-2.5,0", Scale="1,1,1") {
    rotationMode=BMM_Additive
    rotationSpace=BCS_BoneSpace
    translationMode=BMM_Additive
    translationSpace=BCS_BoneSpace
    alphaBoolBlend {blendOption=Linear}
    boneToModify {boneName=dyn_rear_skirt_corner_l_MSTR}
    rotation {yaw=-2.5}
    translation {y=-1}
   }
   n42 = "Transform (Modify) Bone - Bone: dyn_rear_skirt_l_MSTR"(ComponentPose=@n41, Alpha=0.5, Translation="0,-1,0", Rotation="0,-2.5,0", Scale="1,1,1") {
    rotationMode=BMM_Additive
    rotationSpace=BCS_BoneSpace
    translationMode=BMM_Additive
    translationSpace=BCS_BoneSpace
    alphaBoolBlend {blendOption=Linear}
    boneToModify {boneName=dyn_rear_skirt_l_MSTR}
    rotation {yaw=-2.5}
    translation {y=-1}
   }
   n43 = "Transform (Modify) Bone - Bone: dyn_rear_skirt_r_MSTR"(ComponentPose=@n42, Alpha=0.5, Translation="0,1,0", Rotation="0,-2.5,0", Scale="1,1,1") {
    rotationMode=BMM_Additive
    rotationSpace=BCS_BoneSpace
    translationMode=BMM_Additive
    translationSpace=BCS_BoneSpace
    alphaBoolBlend {blendOption=Linear}
    boneToModify {boneName=dyn_rear_skirt_r_MSTR}
    rotation {yaw=-2.5}
    translation {y=1}
   }
   n44 = "Transform (Modify) Bone - Bone: dyn_rear_skirt_corner_r_MSTR"(ComponentPose=@n43, Alpha=0.5, Translation="0,1,0", Rotation="0,-2.5,0", Scale="1,1,1") {
    rotationMode=BMM_Additive
    rotationSpace=BCS_BoneSpace
    translationMode=BMM_Additive
    translationSpace=BCS_BoneSpace
    alphaBoolBlend {blendOption=Linear}
    boneToModify {boneName=dyn_rear_skirt_corner_r_MSTR}
    rotation {yaw=-2.5}
    translation {y=1}
   }
   n45 = "Transform (Modify) Bone - Bone: dyn_side_skirt_r_MSTR"(ComponentPose=@n44, Alpha=1, Translation="0,1,0", Rotation="0,-2.5,0", Scale="1,1,1") {
    rotationMode=BMM_Additive
    rotationSpace=BCS_BoneSpace
    translationMode=BMM_Additive
    translationSpace=BCS_BoneSpace
    alphaBoolBlend {blendOption=Linear}
    boneToModify {boneName=dyn_side_skirt_r_MSTR}
    rotation {yaw=-2.5}
    translation {y=1}
   }
   n46 = "Transform (Modify) Bone - Bone: dyn_front_skirt_r_MSTR"(ComponentPose=@n45, Alpha=1, Translation="0,1,0", Rotation="0,-2.5,0", Scale="1,1,1") {
    rotationMode=BMM_Additive
    rotationSpace=BCS_BoneSpace
    translationMode=BMM_Additive
    translationSpace=BCS_BoneSpace
    alphaBoolBlend {blendOption=Linear}
    boneToModify {boneName=dyn_front_skirt_r_MSTR}
    rotation {yaw=-2.5}
    translation {y=1}
   }
   n47 = "Transform (Modify) Bone - Bone: dyn_front_skirt_corner_r_MSTR"(ComponentPose=@n46, Alpha=1, Translation="0,1,0", Rotation="0,-2.5,0", Scale="1,1,1") {
    rotationMode=BMM_Additive
    rotationSpace=BCS_BoneSpace
    translationMode=BMM_Additive
    translationSpace=BCS_BoneSpace
    alphaBoolBlend {blendOption=Linear}
    boneToModify {boneName=dyn_front_skirt_corner_r_MSTR}
    rotation {yaw=-2.5}
    translation {y=1}
   }
   n48 = "Constraint - MSTR_hand_cloth_2_r (PPPPPPPP)"!(
    ComponentPose=@n47
    Alpha=1
    ConstraintWeights_0=0.5408
    ConstraintWeights_1:float=0.1
    ConstraintWeights_2:float=0.0257
    ConstraintWeights_3:float=0.0001
    ConstraintWeights_4:float=0.0389
    ConstraintWeights_5:float=0.0529
    ConstraintWeights_6:float=0.0601
    ConstraintWeights_7:float=0.1816
   ) -> (Pose) {
    alphaBoolBlend {blendOption=Linear}
    boneToModify {boneName=MSTR_hand_cloth_2_r}
    constraintSetup [
     {targetBone {boneName=upperarm_r}}
     {targetBone {boneName=lowerarm_r}}
     {targetBone {boneName=lowerarm_twist_02_r}}
     {targetBone {boneName=lowerarm_twist_01_r}}
     {targetBone {boneName=deform_elbow_in_r}}
     {targetBone {boneName=deform_elbow_out_r}}
     {targetBone {boneName=upperarm_twist_01_r}}
     {targetBone {boneName=upperarm_twist_02_r}}
    ]
    constraintWeights [1,1,1,1,1,1,1,1]
   }
   n49 = "Constraint - MSTR_hand_cloth_4_r (PPPPPPPP)"!(
    ComponentPose=@n48
    Alpha=1
    ConstraintWeights_0=0.4061
    ConstraintWeights_1:float=0.2148
    ConstraintWeights_2:float=0.112
    ConstraintWeights_3:float=0.0046
    ConstraintWeights_4:float=0.0643
    ConstraintWeights_5:float=0.1257
    ConstraintWeights_6:float=0.0057
    ConstraintWeights_7:float=0.0669
   ) -> (Pose) {
    alphaBoolBlend {blendOption=Linear}
    boneToModify {boneName=MSTR_hand_cloth_4_r}
    constraintSetup [
     {targetBone {boneName=upperarm_r}}
     {targetBone {boneName=lowerarm_r}}
     {targetBone {boneName=lowerarm_twist_02_r}}
     {targetBone {boneName=lowerarm_twist_01_r}}
     {targetBone {boneName=deform_elbow_in_r}}
     {targetBone {boneName=deform_elbow_out_r}}
     {targetBone {boneName=upperarm_twist_01_r}}
     {targetBone {boneName=upperarm_twist_02_r}}
    ]
    constraintWeights [1,1,1,1,1,1,1,1]
   }
   n50 = "Constraint - MSTR_hand_cloth_1_r (PPPPPPP)"!(
    ComponentPose=@n49
    Alpha=1
    ConstraintWeights_0=0.1106
    ConstraintWeights_1:float=0.3086
    ConstraintWeights_2:float=0.4043
    ConstraintWeights_3:float=0.0932
    ConstraintWeights_4:float=0.0239
    ConstraintWeights_5:float=0.0569
    ConstraintWeights_6:float=0.0025
   ) -> (Pose) {
    alphaBoolBlend {blendOption=Linear}
    boneToModify {boneName=MSTR_hand_cloth_1_r}
    constraintSetup [
     {targetBone {boneName=upperarm_r}}
     {targetBone {boneName=lowerarm_r}}
     {targetBone {boneName=lowerarm_twist_02_r}}
     {targetBone {boneName=lowerarm_twist_01_r}}
     {targetBone {boneName=deform_elbow_in_r}}
     {targetBone {boneName=deform_elbow_out_r}}
     {targetBone {boneName=upperarm_twist_02_r}}
    ]
    constraintWeights [1,1,1,1,1,1,1]
   }
   n51 = "Constraint - MSTR_hand_cloth_0_r (PPPPPP)"!(
    ComponentPose=@n50
    Alpha=1
    ConstraintWeights_0=0.0113
    ConstraintWeights_1:float=0.1855
    ConstraintWeights_2:float=0.4753
    ConstraintWeights_3:float=0.326
    ConstraintWeights_4:float=0.001
    ConstraintWeights_5:float=0.0009
   ) -> (Pose) {
    alphaBoolBlend {blendOption=Linear}
    boneToModify {boneName=MSTR_hand_cloth_0_r}
    constraintSetup [
     {targetBone {boneName=upperarm_r}}
     {targetBone {boneName=lowerarm_r}}
     {targetBone {boneName=lowerarm_twist_02_r}}
     {targetBone {boneName=lowerarm_twist_01_r}}
     {targetBone {boneName=deform_elbow_in_r}}
     {targetBone {boneName=deform_elbow_out_r}}
    ]
    constraintWeights [1,1,1,1,1,1]
   }
   n52 = "Constraint - MSTR_hand_cloth_2_l (PPPPPPPP)"!(
    ComponentPose=@n51
    Alpha=1
    ConstraintWeights_0=0.5416
    ConstraintWeights_1:float=0.0992
    ConstraintWeights_2:float=0.0251
    ConstraintWeights_3:float=0
    ConstraintWeights_4:float=0.038
    ConstraintWeights_5:float=0.0536
    ConstraintWeights_6:float=0.0603
    ConstraintWeights_7:float=0.1822
   ) -> (Pose) {
    alphaBoolBlend {blendOption=Linear}
    boneToModify {boneName=MSTR_hand_cloth_2_l}
    constraintSetup [
     {targetBone {boneName=upperarm_l}}
     {targetBone {boneName=lowerarm_l}}
     {targetBone {boneName=lowerarm_twist_02_l}}
     {targetBone {boneName=lowerarm_twist_01_l}}
     {targetBone {boneName=deform_elbow_in_l}}
     {targetBone {boneName=deform_elbow_out_l}}
     {targetBone {boneName=upperarm_twist_01_l}}
     {targetBone {boneName=upperarm_twist_02_l}}
    ]
    constraintWeights [1,1,1,1,1,1,1,1]
   }
   n53 = "Constraint - MSTR_hand_cloth_4_l (PPPPPPPP)"!(
    ComponentPose=@n52
    Alpha=1
    ConstraintWeights_0=0.4064
    ConstraintWeights_1:float=0.2141
    ConstraintWeights_2:float=0.1073
    ConstraintWeights_3:float=0.0034
    ConstraintWeights_4:float=0.0636
    ConstraintWeights_5:float=0.1351
    ConstraintWeights_6:float=0.0048
    ConstraintWeights_7:float=0.0652
   ) -> (Pose) {
    alphaBoolBlend {blendOption=Linear}
    boneToModify {boneName=MSTR_hand_cloth_4_l}
    constraintSetup [
     {targetBone {boneName=upperarm_l}}
     {targetBone {boneName=lowerarm_l}}
     {targetBone {boneName=lowerarm_twist_02_l}}
     {targetBone {boneName=lowerarm_twist_01_l}}
     {targetBone {boneName=deform_elbow_in_l}}
     {targetBone {boneName=deform_elbow_out_l}}
     {targetBone {boneName=upperarm_twist_01_l}}
     {targetBone {boneName=upperarm_twist_02_l}}
    ]
    constraintWeights [1,1,1,1,1,1,1,1]
   }
   n54 = "Constraint - MSTR_hand_cloth_1_l (PPPPPPPP)"!(
    ComponentPose=@n53
    Alpha=1
    ConstraintWeights_0=0.1514
    ConstraintWeights_1:float=0.3008
    ConstraintWeights_2:float=0.3331
    ConstraintWeights_3:float=0.0595
    ConstraintWeights_4:float=0.0311
    ConstraintWeights_5:float=0.0916
    ConstraintWeights_6:float=0.0274
    ConstraintWeights_7:float=0.005
   ) -> (Pose) {
    alphaBoolBlend {blendOption=Linear}
    boneToModify {boneName=MSTR_hand_cloth_1_l}
    constraintSetup [
     {targetBone {boneName=upperarm_l}}
     {targetBone {boneName=lowerarm_l}}
     {targetBone {boneName=lowerarm_twist_02_l}}
     {targetBone {boneName=lowerarm_twist_01_l}}
     {targetBone {boneName=deform_elbow_in_l}}
     {targetBone {boneName=deform_elbow_out_l}}
     {targetBone {boneName=dyn_hand_cloth_11_l}}
     {targetBone {boneName=upperarm_twist_02_l}}
    ]
    constraintWeights [1,1,1,1,1,1,1,1]
   }
   n55 = "Constraint - MSTR_hand_cloth_0_l (PPPPPPPP)"!(
    ComponentPose=@n54
    Alpha=1
    ConstraintWeights_0=0.0283
    ConstraintWeights_1:float=0.2313
    ConstraintWeights_2:float=0.4918
    ConstraintWeights_3:float=0.2249
    ConstraintWeights_4:float=0.0021
    ConstraintWeights_5:float=0.008
    ConstraintWeights_6:float=0.0028
    ConstraintWeights_7:float=0.0108
   ) -> (Pose) {
    alphaBoolBlend {blendOption=Linear}
    boneToModify {boneName=MSTR_hand_cloth_0_l}
    constraintSetup [
     {targetBone {boneName=upperarm_l}}
     {targetBone {boneName=lowerarm_l}}
     {targetBone {boneName=lowerarm_twist_02_l}}
     {targetBone {boneName=lowerarm_twist_01_l}}
     {targetBone {boneName=deform_elbow_in_l}}
     {targetBone {boneName=deform_elbow_out_l}}
     {targetBone {boneName=dyn_hand_cloth_11_l}}
     {targetBone {boneName=dyn_hand_cloth_01_l}}
    ]
    constraintWeights [1,1,1,1,1,1,1,1]
   }
   n56 = "Constraint - MSTR_hand_cloth_3_l (PPPPPP)"!(
    ComponentPose=@n55
    Alpha=1
    ConstraintWeights_0=0
    ConstraintWeights_1:float=0.0707
    ConstraintWeights_2:float=0.0169
    ConstraintWeights_3:float=0.0001
    ConstraintWeights_4:float=0.2924
    ConstraintWeights_5:float=0.6199
   ) -> (Pose) {
    alphaBoolBlend {blendOption=Linear}
    boneToModify {boneName=MSTR_hand_cloth_3_l}
    constraintSetup [
     {targetBone {boneName=upperarm_l}}
     {targetBone {boneName=lowerarm_l}}
     {targetBone {boneName=deform_wrist_dn_l}}
     {targetBone {boneName=deform_wrist_up_l}}
     {targetBone {boneName=lowerarm_twist_02_l}}
     {targetBone {boneName=lowerarm_twist_01_l}}
    ]
    constraintWeights [1,1,1,1,1,1]
   }
   n57 = "Transform (Modify) Bone - Bone: MSTR_hand_cloth_4_l"(ComponentPose=@n56, Alpha=1, Translation="0,0,0", Rotation="0,45,0", Scale="1,1,1") {
    rotationMode=BMM_Additive
    alphaBoolBlend {blendOption=Linear}
    alphaScaleBiasClamp {bInterpResult=true}
    boneToModify {boneName=MSTR_hand_cloth_4_l}
    rotation {yaw=45}
   }
   n58 = "Transform (Modify) Bone - Bone: MSTR_hand_cloth_2_l"(ComponentPose=@n57, Alpha=1, Translation="0,0,0", Rotation="0,45,0", Scale="1,1,1") {
    rotationMode=BMM_Additive
    alphaBoolBlend {blendOption=Linear}
    alphaScaleBiasClamp {bInterpResult=true}
    boneToModify {boneName=MSTR_hand_cloth_2_l}
    rotation {yaw=45}
   }
   n59 = "Transform (Modify) Bone - Bone: MSTR_hand_cloth_1_l"(ComponentPose=@n58, Alpha=1, Translation="0,0,0", Rotation="0,45,0", Scale="1,1,1") {
    rotationMode=BMM_Additive
    alphaBoolBlend {blendOption=Linear}
    alphaScaleBiasClamp {bInterpResult=true}
    boneToModify {boneName=MSTR_hand_cloth_1_l}
    rotation {yaw=45}
   }
   n60 = "Transform (Modify) Bone - Bone: MSTR_hand_cloth_0_l"(ComponentPose=@n59, Alpha=1, Translation="0,0,0", Rotation="0,45,0", Scale="1,1,1") {
    rotationMode=BMM_Additive
    alphaBoolBlend {blendOption=Linear}
    alphaScaleBiasClamp {bInterpResult=true}
    boneToModify {boneName=MSTR_hand_cloth_0_l}
    rotation {yaw=45}
   }
   n61 = "Transform (Modify) Bone - Bone: MSTR_hand_cloth_3_l"(ComponentPose=@n60, Alpha=1, Translation="0,0,0", Rotation="0,45,0", Scale="1,1,1") {
    rotationMode=BMM_Additive
    alphaBoolBlend {blendOption=Linear}
    alphaScaleBiasClamp {bInterpResult=true}
    boneToModify {boneName=MSTR_hand_cloth_3_l}
    rotation {yaw=45}
   }
   n62 = "Transform (Modify) Bone - Bone: MSTR_hand_cloth_2_r"(ComponentPose=@n61, Alpha=1, Translation="0,0,0", Rotation="0,-55,0", Scale="1,1,1") {
    rotationMode=BMM_Additive
    alphaBoolBlend {blendOption=Linear}
    alphaScaleBiasClamp {bInterpResult=true}
    boneToModify {boneName=MSTR_hand_cloth_2_r}
    rotation {yaw=-55}
   }
   n63 = "Transform (Modify) Bone - Bone: MSTR_hand_cloth_3_r"(ComponentPose=@n62, Alpha=1, Translation="0,0,0", Rotation="0,-60,0", Scale="1,1,1") {
    rotationMode=BMM_Additive
    alphaBoolBlend {blendOption=Linear}
    alphaScaleBiasClamp {bInterpResult=true}
    boneToModify {boneName=MSTR_hand_cloth_3_r}
    rotation {yaw=-60}
   }
   n64 = "Transform (Modify) Bone - Bone: MSTR_hand_cloth_1_r"(ComponentPose=@n63, Alpha=1, Translation="0,0,0", Rotation="0,-35,0", Scale="1,1,1") {
    rotationMode=BMM_Additive
    alphaBoolBlend {blendOption=Linear}
    alphaScaleBiasClamp {bInterpResult=true}
    boneToModify {boneName=MSTR_hand_cloth_1_r}
    rotation {yaw=-35}
   }
   n65 = "Transform (Modify) Bone - Bone: MSTR_hand_cloth_0_r"(ComponentPose=@n64, Alpha=1, Translation="0,0,0", Rotation="0,-35,0", Scale="1,1,1") {
    rotationMode=BMM_Additive
    alphaBoolBlend {blendOption=Linear}
    alphaScaleBiasClamp {bInterpResult=true}
    boneToModify {boneName=MSTR_hand_cloth_0_r}
    rotation {yaw=-35}
   }
   n66 = "Transform (Modify) Bone - Bone: MSTR_hand_cloth_4_r"(ComponentPose=@n65, Alpha=1, Translation="0,0,0", Rotation="0,-60,0", Scale="1,1,1") {
    rotationMode=BMM_Additive
    alphaBoolBlend {blendOption=Linear}
    alphaScaleBiasClamp {bInterpResult=true}
    boneToModify {boneName=MSTR_hand_cloth_4_r}
    rotation {yaw=-60}
   }
   Save cached pose 'Offsets'(Pose=Component To Local(ComponentPose=@n66)) {cacheName=Offsets}
   n67 = Anim Dynamics - Bone: dyn_front_skirt_corner_l_02(ComponentPose=Local To Component(LocalPose=Use cached pose 'Offsets'()), Alpha=1, ExternalForce="0,0,0") {
    angularDampingOverride=1
    angularSpringConstant=4000
    bAngularSpring=true
    bOverrideAngularDamping=true
    bUseGravityOverride=true
    gravityScale=0.35
    simulationSpace=RootRelative
    alphaBoolBlend {blendOption=Linear}
    boundBone {boneName=dyn_front_skirt_corner_l_02}
    componentAppliedLinearAccClamp=<1000 1000 1000>
    componentLinearAccScale=<1 1 1>
    componentLinearVelScale=<1 1 1>
    gravityOverride {z=-2000}
    physicsBodyDefinitions [{
      boundBone {boneName=dyn_front_skirt_corner_l_02}
      boxExtents=<30 40 40>
      constraintSetup {angularLimitsMax=<20 20 2>, angularLimitsMin=<-20 -20 -45>, angularTarget=<1 0 0>}
      localJointOffset=<7 0 0>
     }]
   }
   n68 = Anim Dynamics - Bone: dyn_front_skirt_corner_l_03(ComponentPose=@n67, Alpha=1, ExternalForce="0,0,0") {
    angularDampingOverride=1
    angularSpringConstant=4000
    bAngularSpring=true
    bOverrideAngularDamping=true
    bUseGravityOverride=true
    gravityScale=0.35
    simulationSpace=RootRelative
    alphaBoolBlend {blendOption=Linear}
    boundBone {boneName=dyn_front_skirt_corner_l_03}
    componentAppliedLinearAccClamp=<1000 1000 1000>
    componentLinearAccScale=<1 1 1>
    componentLinearVelScale=<1 1 1>
    gravityOverride {z=-2000}
    physicsBodyDefinitions [{
      boundBone {boneName=dyn_front_skirt_corner_l_03}
      boxExtents=<30 40 40>
      constraintSetup {angularLimitsMax=<20 20 2>, angularLimitsMin=<-20 -20 -45>, angularTarget=<1 0 0>}
      localJointOffset=<7 0 0>
     }]
   }
   n69 = Anim Dynamics - Bone: dyn_front_skirt_corner_l_04(ComponentPose=@n68, Alpha=1, ExternalForce="0,0,0") {
    angularDampingOverride=1
    angularSpringConstant=4000
    bAngularSpring=true
    bOverrideAngularDamping=true
    bUseGravityOverride=true
    gravityScale=0.35
    simulationSpace=RootRelative
    alphaBoolBlend {blendOption=Linear}
    boundBone {boneName=dyn_front_skirt_corner_l_04}
    componentAppliedLinearAccClamp=<1000 1000 1000>
    componentLinearAccScale=<1 1 1>
    componentLinearVelScale=<1 1 1>
    gravityOverride {z=-2000}
    physicsBodyDefinitions [{
      boundBone {boneName=dyn_front_skirt_corner_l_04}
      boxExtents=<30 40 40>
      constraintSetup {angularLimitsMax=<20 20 2>, angularLimitsMin=<-20 -20 -45>, angularTarget=<1 0 0>}
      localJointOffset=<7 0 0>
     }]
   }
   n70 = Anim Dynamics - Bone: dyn_front_skirt_corner_l_05(ComponentPose=@n69, Alpha=1, ExternalForce="0,0,0") {
    angularDampingOverride=1
    angularSpringConstant=4000
    bAngularSpring=true
    bOverrideAngularDamping=true
    bUseGravityOverride=true
    gravityScale=0.35
    simulationSpace=RootRelative
    alphaBoolBlend {blendOption=Linear}
    boundBone {boneName=dyn_front_skirt_corner_l_05}
    componentAppliedLinearAccClamp=<1000 1000 1000>
    componentLinearAccScale=<1 1 1>
    componentLinearVelScale=<1 1 1>
    gravityOverride {z=-2000}
    physicsBodyDefinitions [{
      boundBone {boneName=dyn_front_skirt_corner_l_05}
      boxExtents=<30 40 40>
      constraintSetup {angularLimitsMax=<20 20 2>, angularLimitsMin=<-20 -20 -45>, angularTarget=<1 0 0>}
      localJointOffset=<7 0 0>
     }]
   }
   n71 = Anim Dynamics - Bone: dyn_rear_skirt_l_02(ComponentPose=@n70, Alpha=1, ExternalForce="0,0,0") {
    angularDampingOverride=1
    angularSpringConstant=4000
    bAngularSpring=true
    bOverrideAngularDamping=true
    bUseGravityOverride=true
    gravityScale=0.35
    simulationSpace=RootRelative
    alphaBoolBlend {blendOption=Linear}
    boundBone {boneName=dyn_rear_skirt_l_02}
    componentAppliedLinearAccClamp=<1000 1000 1000>
    componentLinearAccScale=<1 1 1>
    componentLinearVelScale=<1 1 1>
    gravityOverride {z=-2000}
    physicsBodyDefinitions [{
      boundBone {boneName=dyn_rear_skirt_l_02}
      boxExtents=<30 40 40>
      constraintSetup {angularLimitsMax=<20 20 2>, angularLimitsMin=<-20 -20 -45>, angularTarget=<1 0 0>}
      localJointOffset=<7 0 0>
     }]
   }
   n72 = Anim Dynamics - Bone: dyn_rear_skirt_l_03(ComponentPose=@n71, Alpha=1, ExternalForce="0,0,0") {
    angularDampingOverride=1
    angularSpringConstant=4000
    bAngularSpring=true
    bOverrideAngularDamping=true
    bUseGravityOverride=true
    gravityScale=0.35
    simulationSpace=RootRelative
    alphaBoolBlend {blendOption=Linear}
    boundBone {boneName=dyn_rear_skirt_l_03}
    componentAppliedLinearAccClamp=<1000 1000 1000>
    componentLinearAccScale=<1 1 1>
    componentLinearVelScale=<1 1 1>
    gravityOverride {z=-2000}
    physicsBodyDefinitions [{
      boundBone {boneName=dyn_rear_skirt_l_03}
      boxExtents=<30 40 40>
      constraintSetup {angularLimitsMax=<20 20 2>, angularLimitsMin=<-20 -20 -45>, angularTarget=<1 0 0>}
      localJointOffset=<7 0 0>
     }]
   }
   n73 = Anim Dynamics - Bone: dyn_rear_skirt_l_04(ComponentPose=@n72, Alpha=1, ExternalForce="0,0,0") {
    angularDampingOverride=1
    angularSpringConstant=4000
    bAngularSpring=true
    bOverrideAngularDamping=true
    bUseGravityOverride=true
    gravityScale=0.35
    simulationSpace=RootRelative
    alphaBoolBlend {blendOption=Linear}
    boundBone {boneName=dyn_rear_skirt_l_04}
    componentAppliedLinearAccClamp=<1000 1000 1000>
    componentLinearAccScale=<1 1 1>
    componentLinearVelScale=<1 1 1>
    gravityOverride {z=-2000}
    physicsBodyDefinitions [{
      boundBone {boneName=dyn_rear_skirt_l_04}
      boxExtents=<30 40 40>
      constraintSetup {angularLimitsMax=<20 20 2>, angularLimitsMin=<-20 -20 -45>, angularTarget=<1 0 0>}
      localJointOffset=<7 0 0>
     }]
   }
   n74 = Anim Dynamics - Bone: dyn_rear_skirt_l_05(ComponentPose=@n73, Alpha=1, ExternalForce="0,0,0") {
    angularDampingOverride=1
    angularSpringConstant=4000
    bAngularSpring=true
    bOverrideAngularDamping=true
    bUseGravityOverride=true
    gravityScale=0.35
    simulationSpace=RootRelative
    alphaBoolBlend {blendOption=Linear}
    boundBone {boneName=dyn_rear_skirt_l_05}
    componentAppliedLinearAccClamp=<1000 1000 1000>
    componentLinearAccScale=<1 1 1>
    componentLinearVelScale=<1 1 1>
    gravityOverride {z=-2000}
    physicsBodyDefinitions [{
      boundBone {boneName=dyn_rear_skirt_l_05}
      boxExtents=<30 40 40>
      constraintSetup {angularLimitsMax=<20 20 2>, angularLimitsMin=<-20 -20 -45>, angularTarget=<1 0 0>}
      localJointOffset=<7 0 0>
     }]
   }
   n75 = Anim Dynamics - Bone: dyn_rear_skirt_corner_l_02(ComponentPose=@n74, Alpha=1, ExternalForce="0,0,0") {
    angularDampingOverride=1
    angularSpringConstant=4000
    bAngularSpring=true
    bOverrideAngularDamping=true
    bUseGravityOverride=true
    gravityScale=0.35
    simulationSpace=RootRelative
    alphaBoolBlend {blendOption=Linear}
    boundBone {boneName=dyn_rear_skirt_corner_l_02}
    componentAppliedLinearAccClamp=<1000 1000 1000>
    componentLinearAccScale=<1 1 1>
    componentLinearVelScale=<1 1 1>
    gravityOverride {z=-2000}
    physicsBodyDefinitions [{
      boundBone {boneName=dyn_rear_skirt_corner_l_02}
      boxExtents=<30 40 40>
      constraintSetup {angularLimitsMax=<20 20 2>, angularLimitsMin=<-20 -20 -45>, angularTarget=<1 0 0>}
      localJointOffset=<7 0 0>
     }]
   }
   n76 = Anim Dynamics - Bone: dyn_rear_skirt_corner_l_03(ComponentPose=@n75, Alpha=1, ExternalForce="0,0,0") {
    angularDampingOverride=1
    angularSpringConstant=4000
    bAngularSpring=true
    bOverrideAngularDamping=true
    bUseGravityOverride=true
    gravityScale=0.35
    simulationSpace=RootRelative
    alphaBoolBlend {blendOption=Linear}
    boundBone {boneName=dyn_rear_skirt_corner_l_03}
    componentAppliedLinearAccClamp=<1000 1000 1000>
    componentLinearAccScale=<1 1 1>
    componentLinearVelScale=<1 1 1>
    gravityOverride {z=-2000}
    physicsBodyDefinitions [{
      boundBone {boneName=dyn_rear_skirt_corner_l_03}
      boxExtents=<30 40 40>
      constraintSetup {angularLimitsMax=<20 20 2>, angularLimitsMin=<-20 -20 -45>, angularTarget=<1 0 0>}
      localJointOffset=<7 0 0>
     }]
   }
   n77 = Anim Dynamics - Bone: dyn_rear_skirt_corner_l_04(ComponentPose=@n76, Alpha=1, ExternalForce="0,0,0") {
    angularDampingOverride=1
    angularSpringConstant=4000
    bAngularSpring=true
    bOverrideAngularDamping=true
    bUseGravityOverride=true
    gravityScale=0.35
    simulationSpace=RootRelative
    alphaBoolBlend {blendOption=Linear}
    boundBone {boneName=dyn_rear_skirt_corner_l_04}
    componentAppliedLinearAccClamp=<1000 1000 1000>
    componentLinearAccScale=<1 1 1>
    componentLinearVelScale=<1 1 1>
    gravityOverride {z=-2000}
    physicsBodyDefinitions [{
      boundBone {boneName=dyn_rear_skirt_corner_l_04}
      boxExtents=<30 40 40>
      constraintSetup {angularLimitsMax=<20 20 2>, angularLimitsMin=<-20 -20 -45>, angularTarget=<1 0 0>}
      localJointOffset=<7 0 0>
     }]
   }
   n78 = Anim Dynamics - Bone: dyn_rear_skirt_corner_l_05(ComponentPose=@n77, Alpha=1, ExternalForce="0,0,0") {
    angularDampingOverride=1
    angularSpringConstant=4000
    bAngularSpring=true
    bOverrideAngularDamping=true
    bUseGravityOverride=true
    gravityScale=0.35
    simulationSpace=RootRelative
    alphaBoolBlend {blendOption=Linear}
    boundBone {boneName=dyn_rear_skirt_corner_l_05}
    componentAppliedLinearAccClamp=<1000 1000 1000>
    componentLinearAccScale=<1 1 1>
    componentLinearVelScale=<1 1 1>
    gravityOverride {z=-2000}
    physicsBodyDefinitions [{
      boundBone {boneName=dyn_rear_skirt_corner_l_05}
      boxExtents=<30 40 40>
      constraintSetup {angularLimitsMax=<20 20 2>, angularLimitsMin=<-20 -20 -45>, angularTarget=<1 0 0>}
      localJointOffset=<7 0 0>
     }]
   }
   n79 = Anim Dynamics - Bone: dyn_side_skirt_l_02(ComponentPose=@n78, Alpha=1, ExternalForce="0,0,0") {
    angularDampingOverride=1
    angularSpringConstant=4000
    bAngularSpring=true
    bOverrideAngularDamping=true
    bUseGravityOverride=true
    gravityScale=0.35
    simulationSpace=RootRelative
    alphaBoolBlend {blendOption=Linear}
    boundBone {boneName=dyn_side_skirt_l_02}
    componentAppliedLinearAccClamp=<1000 1000 1000>
    componentLinearAccScale=<1 1 1>
    componentLinearVelScale=<1 1 1>
    gravityOverride {z=-2000}
    physicsBodyDefinitions [{
      boundBone {boneName=dyn_side_skirt_l_02}
      boxExtents=<30 40 40>
      constraintSetup {angularLimitsMax=<20 20 2>, angularLimitsMin=<-20 -20 -45>, angularTarget=<1 0 0>}
      localJointOffset=<7 0 0>
     }]
   }
   n80 = Anim Dynamics - Bone: dyn_side_skirt_l_03(ComponentPose=@n79, Alpha=1, ExternalForce="0,0,0") {
    angularDampingOverride=1
    angularSpringConstant=4000
    bAngularSpring=true
    bOverrideAngularDamping=true
    bUseGravityOverride=true
    gravityScale=0.35
    simulationSpace=RootRelative
    alphaBoolBlend {blendOption=Linear}
    boundBone {boneName=dyn_side_skirt_l_03}
    componentAppliedLinearAccClamp=<1000 1000 1000>
    componentLinearAccScale=<1 1 1>
    componentLinearVelScale=<1 1 1>
    gravityOverride {z=-2000}
    physicsBodyDefinitions [{
      boundBone {boneName=dyn_side_skirt_l_03}
      boxExtents=<30 40 40>
      constraintSetup {angularLimitsMax=<20 20 2>, angularLimitsMin=<-20 -20 -45>, angularTarget=<1 0 0>}
      localJointOffset=<7 0 0>
     }]
   }
   n81 = Anim Dynamics - Bone: dyn_side_skirt_l_04(ComponentPose=@n80, Alpha=1, ExternalForce="0,0,0") {
    angularDampingOverride=1
    angularSpringConstant=4000
    bAngularSpring=true
    bOverrideAngularDamping=true
    bUseGravityOverride=true
    gravityScale=0.35
    simulationSpace=RootRelative
    alphaBoolBlend {blendOption=Linear}
    boundBone {boneName=dyn_side_skirt_l_04}
    componentAppliedLinearAccClamp=<1000 1000 1000>
    componentLinearAccScale=<1 1 1>
    componentLinearVelScale=<1 1 1>
    gravityOverride {z=-2000}
    physicsBodyDefinitions [{
      boundBone {boneName=dyn_side_skirt_l_04}
      boxExtents=<30 40 40>
      constraintSetup {angularLimitsMax=<20 20 2>, angularLimitsMin=<-20 -20 -45>, angularTarget=<1 0 0>}
      localJointOffset=<7 0 0>
     }]
   }
   n82 = Anim Dynamics - Bone: dyn_side_skirt_l_05(ComponentPose=@n81, Alpha=1, ExternalForce="0,0,0") {
    angularDampingOverride=1
    angularSpringConstant=4000
    bAngularSpring=true
    bOverrideAngularDamping=true
    bUseGravityOverride=true
    gravityScale=0.35
    simulationSpace=RootRelative
    alphaBoolBlend {blendOption=Linear}
    boundBone {boneName=dyn_side_skirt_l_05}
    componentAppliedLinearAccClamp=<1000 1000 1000>
    componentLinearAccScale=<1 1 1>
    componentLinearVelScale=<1 1 1>
    gravityOverride {z=-2000}
    physicsBodyDefinitions [{
      boundBone {boneName=dyn_side_skirt_l_05}
      boxExtents=<30 40 40>
      constraintSetup {angularLimitsMax=<20 20 2>, angularLimitsMin=<-20 -20 -45>, angularTarget=<1 0 0>}
      localJointOffset=<7 0 0>
     }]
   }
   n83 = Anim Dynamics - Bone: dyn_side_skirt_r_02(ComponentPose=@n82, Alpha=1, ExternalForce="0,0,0") {
    angularDampingOverride=1
    angularSpringConstant=4000
    bAngularSpring=true
    bOverrideAngularDamping=true
    bUseGravityOverride=true
    gravityScale=0.35
    simulationSpace=RootRelative
    alphaBoolBlend {blendOption=Linear}
    boundBone {boneName=dyn_side_skirt_r_02}
    componentAppliedLinearAccClamp=<1000 1000 1000>
    componentLinearAccScale=<1 1 1>
    componentLinearVelScale=<1 1 1>
    gravityOverride {z=-2000}
    physicsBodyDefinitions [{
      boundBone {boneName=dyn_side_skirt_r_02}
      boxExtents=<30 40 40>
      constraintSetup {angularLimitsMax=<20 20 2>, angularLimitsMin=<-20 -20 -45>, angularTarget=<1 0 0>}
      localJointOffset=<-7 0 0>
     }]
   }
   n84 = Anim Dynamics - Bone: dyn_side_skirt_r_03(ComponentPose=@n83, Alpha=1, ExternalForce="0,0,0") {
    angularDampingOverride=1
    angularSpringConstant=4000
    bAngularSpring=true
    bOverrideAngularDamping=true
    bUseGravityOverride=true
    gravityScale=0.35
    simulationSpace=RootRelative
    alphaBoolBlend {blendOption=Linear}
    boundBone {boneName=dyn_side_skirt_r_03}
    componentAppliedLinearAccClamp=<1000 1000 1000>
    componentLinearAccScale=<1 1 1>
    componentLinearVelScale=<1 1 1>
    gravityOverride {z=-2000}
    physicsBodyDefinitions [{
      boundBone {boneName=dyn_side_skirt_r_03}
      boxExtents=<30 40 40>
      constraintSetup {angularLimitsMax=<20 20 2>, angularLimitsMin=<-20 -20 -45>, angularTarget=<1 0 0>}
      localJointOffset=<-7 0 0>
     }]
   }
   n85 = Anim Dynamics - Bone: dyn_side_skirt_r_04(ComponentPose=@n84, Alpha=1, ExternalForce="0,0,0") {
    angularDampingOverride=1
    angularSpringConstant=4000
    bAngularSpring=true
    bOverrideAngularDamping=true
    bUseGravityOverride=true
    gravityScale=0.35
    simulationSpace=RootRelative
    alphaBoolBlend {blendOption=Linear}
    boundBone {boneName=dyn_side_skirt_r_04}
    componentAppliedLinearAccClamp=<1000 1000 1000>
    componentLinearAccScale=<1 1 1>
    componentLinearVelScale=<1 1 1>
    gravityOverride {z=-2000}
    physicsBodyDefinitions [{
      boundBone {boneName=dyn_side_skirt_r_04}
      boxExtents=<30 40 40>
      constraintSetup {angularLimitsMax=<20 20 2>, angularLimitsMin=<-20 -20 -45>, angularTarget=<1 0 0>}
      localJointOffset=<-7 0 0>
     }]
   }
   n86 = Anim Dynamics - Bone: dyn_side_skirt_r_05(ComponentPose=@n85, Alpha=1, ExternalForce="0,0,0") {
    angularDampingOverride=1
    angularSpringConstant=4000
    bAngularSpring=true
    bOverrideAngularDamping=true
    bUseGravityOverride=true
    gravityScale=0.35
    simulationSpace=RootRelative
    alphaBoolBlend {blendOption=Linear}
    boundBone {boneName=dyn_side_skirt_r_05}
    componentAppliedLinearAccClamp=<1000 1000 1000>
    componentLinearAccScale=<1 1 1>
    componentLinearVelScale=<1 1 1>
    gravityOverride {z=-2000}
    physicsBodyDefinitions [{
      boundBone {boneName=dyn_side_skirt_r_05}
      boxExtents=<30 40 40>
      constraintSetup {angularLimitsMax=<20 20 2>, angularLimitsMin=<-20 -20 -45>, angularTarget=<1 0 0>}
      localJointOffset=<-7 0 0>
     }]
   }
   n87 = Anim Dynamics - Bone: dyn_front_skirt_r_02(ComponentPose=@n86, Alpha=1, ExternalForce="0,0,0") {
    angularDampingOverride=1
    angularSpringConstant=4000
    bAngularSpring=true
    bOverrideAngularDamping=true
    bUseGravityOverride=true
    gravityScale=0.35
    simulationSpace=RootRelative
    alphaBoolBlend {blendOption=Linear}
    boundBone {boneName=dyn_front_skirt_r_02}
    componentAppliedLinearAccClamp=<1000 1000 1000>
    componentLinearAccScale=<1 1 1>
    componentLinearVelScale=<1 1 1>
    gravityOverride {z=-2000}
    physicsBodyDefinitions [{
      boundBone {boneName=dyn_front_skirt_r_02}
      boxExtents=<30 40 40>
      constraintSetup {angularLimitsMax=<20 20 2>, angularLimitsMin=<-20 -20 -45>, angularTarget=<1 0 0>}
      localJointOffset=<-7 0 0>
     }]
   }
   n88 = Anim Dynamics - Bone: dyn_front_skirt_r_03(ComponentPose=@n87, Alpha=1, ExternalForce="0,0,0") {
    angularDampingOverride=1
    angularSpringConstant=4000
    bAngularSpring=true
    bOverrideAngularDamping=true
    bUseGravityOverride=true
    gravityScale=0.35
    simulationSpace=RootRelative
    alphaBoolBlend {blendOption=Linear}
    boundBone {boneName=dyn_front_skirt_r_03}
    componentAppliedLinearAccClamp=<1000 1000 1000>
    componentLinearAccScale=<1 1 1>
    componentLinearVelScale=<1 1 1>
    gravityOverride {z=-2000}
    physicsBodyDefinitions [{
      boundBone {boneName=dyn_front_skirt_r_03}
      boxExtents=<30 40 40>
      constraintSetup {angularLimitsMax=<20 20 2>, angularLimitsMin=<-20 -20 -45>, angularTarget=<1 0 0>}
      localJointOffset=<-7 0 0>
     }]
   }
   n89 = Anim Dynamics - Bone: dyn_front_skirt_r_04(ComponentPose=@n88, Alpha=1, ExternalForce="0,0,0") {
    angularDampingOverride=1
    angularSpringConstant=4000
    bAngularSpring=true
    bOverrideAngularDamping=true
    bUseGravityOverride=true
    gravityScale=0.35
    simulationSpace=RootRelative
    alphaBoolBlend {blendOption=Linear}
    boundBone {boneName=dyn_front_skirt_r_04}
    componentAppliedLinearAccClamp=<1000 1000 1000>
    componentLinearAccScale=<1 1 1>
    componentLinearVelScale=<1 1 1>
    gravityOverride {z=-2000}
    physicsBodyDefinitions [{
      boundBone {boneName=dyn_front_skirt_r_04}
      boxExtents=<30 40 40>
      constraintSetup {angularLimitsMax=<20 20 2>, angularLimitsMin=<-20 -20 -45>, angularTarget=<1 0 0>}
      localJointOffset=<-7 0 0>
     }]
   }
   n90 = Anim Dynamics - Bone: dyn_front_skirt_r_05(ComponentPose=@n89, Alpha=1, ExternalForce="0,0,0") {
    angularDampingOverride=1
    angularSpringConstant=4000
    bAngularSpring=true
    bOverrideAngularDamping=true
    bUseGravityOverride=true
    gravityScale=0.35
    simulationSpace=RootRelative
    alphaBoolBlend {blendOption=Linear}
    boundBone {boneName=dyn_front_skirt_r_05}
    componentAppliedLinearAccClamp=<1000 1000 1000>
    componentLinearAccScale=<1 1 1>
    componentLinearVelScale=<1 1 1>
    gravityOverride {z=-2000}
    physicsBodyDefinitions [{
      boundBone {boneName=dyn_front_skirt_r_05}
      boxExtents=<30 40 40>
      constraintSetup {angularLimitsMax=<20 20 2>, angularLimitsMin=<-20 -20 -45>, angularTarget=<1 0 0>}
      localJointOffset=<-7 0 0>
     }]
   }
   n91 = Anim Dynamics - Bone: dyn_front_skirt_corner_r_02(ComponentPose=@n90, Alpha=1, ExternalForce="0,0,0") {
    angularDampingOverride=1
    angularSpringConstant=4000
    bAngularSpring=true
    bOverrideAngularDamping=true
    bUseGravityOverride=true
    gravityScale=0.35
    simulationSpace=RootRelative
    alphaBoolBlend {blendOption=Linear}
    boundBone {boneName=dyn_front_skirt_corner_r_02}
    componentAppliedLinearAccClamp=<1000 1000 1000>
    componentLinearAccScale=<1 1 1>
    componentLinearVelScale=<1 1 1>
    gravityOverride {z=-2000}
    physicsBodyDefinitions [{
      boundBone {boneName=dyn_front_skirt_corner_r_02}
      boxExtents=<30 40 40>
      constraintSetup {angularLimitsMax=<20 20 2>, angularLimitsMin=<-20 -20 -45>, angularTarget=<1 0 0>}
      localJointOffset=<-7 0 0>
     }]
   }
   n92 = Anim Dynamics - Bone: dyn_front_skirt_corner_r_03(ComponentPose=@n91, Alpha=1, ExternalForce="0,0,0") {
    angularDampingOverride=1
    angularSpringConstant=4000
    bAngularSpring=true
    bOverrideAngularDamping=true
    bUseGravityOverride=true
    gravityScale=0.35
    simulationSpace=RootRelative
    alphaBoolBlend {blendOption=Linear}
    boundBone {boneName=dyn_front_skirt_corner_r_03}
    componentAppliedLinearAccClamp=<1000 1000 1000>
    componentLinearAccScale=<1 1 1>
    componentLinearVelScale=<1 1 1>
    gravityOverride {z=-2000}
    physicsBodyDefinitions [{
      boundBone {boneName=dyn_front_skirt_corner_r_03}
      boxExtents=<30 40 40>
      constraintSetup {angularLimitsMax=<20 20 2>, angularLimitsMin=<-20 -20 -45>, angularTarget=<1 0 0>}
      localJointOffset=<-7 0 0>
     }]
   }
   n93 = Anim Dynamics - Bone: dyn_front_skirt_corner_r_04(ComponentPose=@n92, Alpha=1, ExternalForce="0,0,0") {
    angularDampingOverride=1
    angularSpringConstant=4000
    bAngularSpring=true
    bOverrideAngularDamping=true
    bUseGravityOverride=true
    gravityScale=0.35
    simulationSpace=RootRelative
    alphaBoolBlend {blendOption=Linear}
    boundBone {boneName=dyn_front_skirt_corner_r_04}
    componentAppliedLinearAccClamp=<1000 1000 1000>
    componentLinearAccScale=<1 1 1>
    componentLinearVelScale=<1 1 1>
    gravityOverride {z=-2000}
    physicsBodyDefinitions [{
      boundBone {boneName=dyn_front_skirt_corner_r_04}
      boxExtents=<30 40 40>
      constraintSetup {angularLimitsMax=<20 20 2>, angularLimitsMin=<-20 -20 -45>, angularTarget=<1 0 0>}
      localJointOffset=<-7 0 0>
     }]
   }
   n94 = Anim Dynamics - Bone: dyn_front_skirt_corner_r_05(ComponentPose=@n93, Alpha=1, ExternalForce="0,0,0") {
    angularDampingOverride=1
    angularSpringConstant=4000
    bAngularSpring=true
    bOverrideAngularDamping=true
    bUseGravityOverride=true
    gravityScale=0.35
    simulationSpace=RootRelative
    alphaBoolBlend {blendOption=Linear}
    boundBone {boneName=dyn_front_skirt_corner_r_05}
    componentAppliedLinearAccClamp=<1000 1000 1000>
    componentLinearAccScale=<1 1 1>
    componentLinearVelScale=<1 1 1>
    gravityOverride {z=-2000}
    physicsBodyDefinitions [{
      boundBone {boneName=dyn_front_skirt_corner_r_05}
      boxExtents=<30 40 40>
      constraintSetup {angularLimitsMax=<20 20 2>, angularLimitsMin=<-20 -20 -45>, angularTarget=<1 0 0>}
      localJointOffset=<-7 0 0>
     }]
   }
   n95 = Anim Dynamics - Bone: dyn_rear_skirt_r_02(ComponentPose=@n94, Alpha=1, ExternalForce="0,0,0") {
    angularDampingOverride=1
    angularSpringConstant=4000
    bAngularSpring=true
    bOverrideAngularDamping=true
    bUseGravityOverride=true
    gravityScale=0.35
    simulationSpace=RootRelative
    alphaBoolBlend {blendOption=Linear}
    boundBone {boneName=dyn_rear_skirt_r_02}
    componentAppliedLinearAccClamp=<1000 1000 1000>
    componentLinearAccScale=<1 1 1>
    componentLinearVelScale=<1 1 1>
    gravityOverride {z=-2000}
    physicsBodyDefinitions [{
      boundBone {boneName=dyn_rear_skirt_r_02}
      boxExtents=<30 40 40>
      constraintSetup {angularLimitsMax=<20 20 2>, angularLimitsMin=<-20 -20 -45>, angularTarget=<1 0 0>}
      localJointOffset=<-7 0 0>
     }]
   }
   n96 = Anim Dynamics - Bone: dyn_rear_skirt_r_03(ComponentPose=@n95, Alpha=1, ExternalForce="0,0,0") {
    angularDampingOverride=1
    angularSpringConstant=4000
    bAngularSpring=true
    bOverrideAngularDamping=true
    bUseGravityOverride=true
    gravityScale=0.35
    simulationSpace=RootRelative
    alphaBoolBlend {blendOption=Linear}
    boundBone {boneName=dyn_rear_skirt_r_03}
    componentAppliedLinearAccClamp=<1000 1000 1000>
    componentLinearAccScale=<1 1 1>
    componentLinearVelScale=<1 1 1>
    gravityOverride {z=-2000}
    physicsBodyDefinitions [{
      boundBone {boneName=dyn_rear_skirt_r_03}
      boxExtents=<30 40 40>
      constraintSetup {angularLimitsMax=<20 20 2>, angularLimitsMin=<-20 -20 -45>, angularTarget=<1 0 0>}
      localJointOffset=<-7 0 0>
     }]
   }
   n97 = Anim Dynamics - Bone: dyn_rear_skirt_r_04(ComponentPose=@n96, Alpha=1, ExternalForce="0,0,0") {
    angularDampingOverride=1
    angularSpringConstant=4000
    bAngularSpring=true
    bOverrideAngularDamping=true
    bUseGravityOverride=true
    gravityScale=0.35
    simulationSpace=RootRelative
    alphaBoolBlend {blendOption=Linear}
    boundBone {boneName=dyn_rear_skirt_r_04}
    componentAppliedLinearAccClamp=<1000 1000 1000>
    componentLinearAccScale=<1 1 1>
    componentLinearVelScale=<1 1 1>
    gravityOverride {z=-2000}
    physicsBodyDefinitions [{
      boundBone {boneName=dyn_rear_skirt_r_04}
      boxExtents=<30 40 40>
      constraintSetup {angularLimitsMax=<20 20 2>, angularLimitsMin=<-20 -20 -45>, angularTarget=<1 0 0>}
      localJointOffset=<-7 0 0>
     }]
   }
   n98 = Anim Dynamics - Bone: dyn_rear_skirt_r_05(ComponentPose=@n97, Alpha=1, ExternalForce="0,0,0") {
    angularDampingOverride=1
    angularSpringConstant=4000
    bAngularSpring=true
    bOverrideAngularDamping=true
    bUseGravityOverride=true
    gravityScale=0.35
    simulationSpace=RootRelative
    alphaBoolBlend {blendOption=Linear}
    boundBone {boneName=dyn_rear_skirt_r_05}
    componentAppliedLinearAccClamp=<1000 1000 1000>
    componentLinearAccScale=<1 1 1>
    componentLinearVelScale=<1 1 1>
    gravityOverride {z=-2000}
    physicsBodyDefinitions [{
      boundBone {boneName=dyn_rear_skirt_r_05}
      boxExtents=<30 40 40>
      constraintSetup {angularLimitsMax=<20 20 2>, angularLimitsMin=<-20 -20 -45>, angularTarget=<1 0 0>}
      localJointOffset=<-7 0 0>
     }]
   }
   n99 = Anim Dynamics - Bone: dyn_rear_skirt_corner_r_02(ComponentPose=@n98, Alpha=1, ExternalForce="0,0,0") {
    angularDampingOverride=1
    angularSpringConstant=4000
    bAngularSpring=true
    bOverrideAngularDamping=true
    bUseGravityOverride=true
    gravityScale=0.35
    simulationSpace=RootRelative
    alphaBoolBlend {blendOption=Linear}
    boundBone {boneName=dyn_rear_skirt_corner_r_02}
    componentAppliedLinearAccClamp=<1000 1000 1000>
    componentLinearAccScale=<1 1 1>
    componentLinearVelScale=<1 1 1>
    gravityOverride {z=-2000}
    physicsBodyDefinitions [{
      boundBone {boneName=dyn_rear_skirt_corner_r_02}
      boxExtents=<30 40 40>
      constraintSetup {angularLimitsMax=<20 20 2>, angularLimitsMin=<-20 -20 -45>, angularTarget=<1 0 0>}
      localJointOffset=<-7 0 0>
     }]
   }
   n100 = Anim Dynamics - Bone: dyn_rear_skirt_corner_r_03(ComponentPose=@n99, Alpha=1, ExternalForce="0,0,0") {
    angularDampingOverride=1
    angularSpringConstant=4000
    bAngularSpring=true
    bOverrideAngularDamping=true
    bUseGravityOverride=true
    gravityScale=0.35
    simulationSpace=RootRelative
    alphaBoolBlend {blendOption=Linear}
    boundBone {boneName=dyn_rear_skirt_corner_r_03}
    componentAppliedLinearAccClamp=<1000 1000 1000>
    componentLinearAccScale=<1 1 1>
    componentLinearVelScale=<1 1 1>
    gravityOverride {z=-2000}
    physicsBodyDefinitions [{
      boundBone {boneName=dyn_rear_skirt_corner_r_03}
      boxExtents=<30 40 40>
      constraintSetup {angularLimitsMax=<20 20 2>, angularLimitsMin=<-20 -20 -45>, angularTarget=<1 0 0>}
      localJointOffset=<-7 0 0>
     }]
   }
   n101 = Anim Dynamics - Bone: dyn_rear_skirt_corner_r_04(ComponentPose=@n100, Alpha=1, ExternalForce="0,0,0") {
    angularDampingOverride=1
    angularSpringConstant=4000
    bAngularSpring=true
    bOverrideAngularDamping=true
    bUseGravityOverride=true
    gravityScale=0.35
    simulationSpace=RootRelative
    alphaBoolBlend {blendOption=Linear}
    boundBone {boneName=dyn_rear_skirt_corner_r_04}
    componentAppliedLinearAccClamp=<1000 1000 1000>
    componentLinearAccScale=<1 1 1>
    componentLinearVelScale=<1 1 1>
    gravityOverride {z=-2000}
    physicsBodyDefinitions [{
      boundBone {boneName=dyn_rear_skirt_corner_r_04}
      boxExtents=<30 40 40>
      constraintSetup {angularLimitsMax=<20 20 2>, angularLimitsMin=<-20 -20 -45>, angularTarget=<1 0 0>}
      localJointOffset=<-7 0 0>
     }]
   }
   n102 = Anim Dynamics - Bone: dyn_rear_skirt_corner_r_05(ComponentPose=@n101, Alpha=1, ExternalForce="0,0,0") {
    angularDampingOverride=1
    angularSpringConstant=4000
    bAngularSpring=true
    bOverrideAngularDamping=true
    bUseGravityOverride=true
    gravityScale=0.35
    simulationSpace=RootRelative
    alphaBoolBlend {blendOption=Linear}
    boundBone {boneName=dyn_rear_skirt_corner_r_05}
    componentAppliedLinearAccClamp=<1000 1000 1000>
    componentLinearAccScale=<1 1 1>
    componentLinearVelScale=<1 1 1>
    gravityOverride {z=-2000}
    physicsBodyDefinitions [{
      boundBone {boneName=dyn_rear_skirt_corner_r_05}
      boxExtents=<30 40 40>
      constraintSetup {angularLimitsMax=<20 20 2>, angularLimitsMin=<-20 -20 -45>, angularTarget=<1 0 0>}
      localJointOffset=<-7 0 0>
     }]
   }
   Save cached pose 'DynamicsSkirt'(Pose=Component To Local(ComponentPose=@n102)) {cacheName=DynamicsSkirt}
   n103 = RigidBody!(ComponentPose=Local To Component(LocalPose=Use cached pose 'Offsets'()), Alpha, OverrideWorldGravity:Vector="0,0,-980", ExternalForce) -> (Pose) {
    bClampLinearTranslationLimitToRefPose=true
    bOverrideWorldGravity=true
    overridePhysicsAsset=&/M_MED_PureCereal_Skirt_Physics
    simulationSpace=BaseBoneSpace
    alphaBoolBlend {blendOption=Linear}
    baseBoneRef {boneName=pelvis}
    overrideWorldGravity {z=-980}
    simSpaceSettings {
     dampingAlpha=0
     maxAngularAcceleration=15
     maxAngularVelocity=3
     maxLinearAcceleration=2000
     maxLinearVelocity=400
     worldAlpha=1
     externalLinearDragV {y=0.5}
    }
   }
   Save cached pose 'RBANSkirt'(Pose=Component To Local(ComponentPose=@n103)) {cacheName=RBANSkirt}
   Save cached pose 'BlendedPoses'(Pose=Blend Poses by bool(BlendPose_0=Use cached pose 'RBANSkirt'(), BlendPose_1=Use cached pose 'DynamicsSkirt'(), bActiveValue=Get bRigidBodyEnabled()) {blendType=Linear}) {cacheName=BlendedPoses}
  ]}
 EventGraph {nodes [
   n1 = Event BlueprintUpdateAnimation()
   > Set isFPSLowerThan30(isFPSLowerThan30:bool="float >= float"(A=@n1.DeltaTimeX, B=0.029)) -> (Output_Get:bool)
   > n2 = Sequence()
   > n3 = "Gravity Override State Function (Parameter Asset) (Nativized)"(
    CharacterPartAnimInstance=Self-Reference()
    GameTimeInSeconds=GetGameTimeInSeconds()
    Parameters=&/M_MED_PureCereal_Parameters
   )
   > Set Gravity Override(Gravity Override=@n3)
   Set MoveGroundCollision(execute=@n2.then_1, MoveGroundCollision:bool=OR Boolean(A=Get bIsPlayingEmote(), B=Get bIsOnGrindRail())) -> (Output_Get:bool)
   n4 = NOT Boolean(A=Get bIsPlayingEmote())
   n5 = Get bIsInFrontEnd()
   > Set FrontEndIdleModify(FrontEndIdleModify:double="To Float (Boolean)"(InBool=AND Boolean(A=@n4, B=@n5))) -> (Output_Get:double)
   n6 = Event BlueprintInitializeAnimation()
   n7 = Branch(execute=[@n6.then, @n8.then], Condition=Get bIsInFrontEnd())
   > Set!(
    execute
    AngularSimSpaceSet_SimSpaceAngularAlpha:float=0.2
    AngularSimSpaceSet_MaxAngularVelocity:float=1500
    AngularSimSpaceSet_MaxAngularAcceleration:float=1500
    AngularSimSpaceSet_ExternalAngularVelocity:Vector="0,0,0"
   ) -> (then, Output_Get:Anim Phys Sim Space Settings Structure)
   Set(
    execute=@n7.else
    AngularSimSpaceSet_SimSpaceAngularAlpha=0.6
    AngularSimSpaceSet_MaxAngularVelocity=3000
    AngularSimSpaceSet_MaxAngularAcceleration=3000
    AngularSimSpaceSet_ExternalAngularVelocity="0,0,0"
   )
   n8 = Event OnCharacterCustomizationLoaded()
  ]}
]
variables [
 AngularSimSpaceSet {
  type=AnimPhysSimSpaceSettings
  default {maxAngularAcceleration=10000, maxAngularVelocity=10000, simSpaceAngularAlpha=0, externalAngularVelocity=<0 0 0>}
 }
 Custom Idle Is Playing Float {type=number}
 FrontEndIdleModify {type=number}
 Gravity Override {type=Vector, default=<0 0 -980>}
 isFPSLowerThan30 {type=boolean}
 MoveGroundCollision {type=boolean}
 Simulation Space {type=AnimPhysSimSpaceType, default=RootRelative}
]
