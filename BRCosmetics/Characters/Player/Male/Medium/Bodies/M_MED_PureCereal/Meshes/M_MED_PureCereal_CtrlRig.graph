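# Control Rig for the M_MED_PureCereal male medium player body, applying procedural deformation correctives.
# Begin Execution runs one shared DeformRig function, configured by per-region toggles, curve flags and offset transforms for areas like shoulders, elbows, wrists, groin, glutes and knees.
# The hierarchy pairs deform bones with base parents.
# It also holds dyn_ chains for hood, arm cloth and skirts, which no other node in this graph drives.
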ControlRigBlueprint /BRCosmetics/Characters/Player/Male/Medium/Bodies/M_MED_PureCereal/Meshes/M_MED_PureCereal_CtrlRig
graphs [
 RigVMFunctionLibrary {links [], nodes []}
 RigVMModel {nodes [
   RigUnit_BeginExecution = &RigUnit_BeginExecution()
   FN_DeformRig_FNC = RigVMFunctionReferenceNode(
    DeformRig_ Curve="(deformRig_l_upperarm_up_alpha_72_#1=False,deformRig_r_upperarm_up_alpha_73_#2=False,deformRig_l_upperarm_fr_alpha_74_#3=False,deformRig_r_upperarm_fr_alpha_75_CB51B8 ...(truncated)"
    DeformRig_Bool="(upperarm_1_#4=True,elbow_3_#5=True,wrist_7_#6=True,groin_8_#7=True,glute_17_#8=True,knee_9_40C408324CA261 ...(truncated)"
    DeformRig_Debug=(upperarm_debug_18_#4=False,wrist_debug_20_#6=False,lat_debug_26_#9=False,groinGlute_debug_27_#7=False)
    DeformRig_Value="(upperarm_up_18_#10=(Rotation=(X=0.0,Y=-0.0,Z=0.0,W=1.0),Translation=(X=0.0,Y=0.0,Z=8.0),Scale3D=(X=0.70,Y=1.0,Z=1.0)),upperarm_fwd_fr_19_#11=(Rotation=(X=0.0,Y=-0.0,Z=0.0,W=1.0),Translation=(X=0.0, ...(truncated)"
    ExecuteContext=@RigUnit_BeginExecution.ExecutePin
   )
  ]}
]
hierarchy [
 root {loc=<0 0 0>, rot=<0 0 0 1>}
 pelvis {parent=root, loc=<0 -0.7071 93.1191>, rot=<0.0127 0.707 -0.0127 -0.707>}
 spine_01 {parent=pelvis, loc=<6.7494 -0.6388 0>, rot=<0 0 0.034 -0.9994>}
 spine_02 {parent=spine_01, loc=<7.2518 -0.035 0>, rot=<0 0 0.0238 -0.9997>}
 spine_03 {parent=spine_02, loc=<6.0104 0 0>, rot=<0 0 0.0149 -0.9999>}
 spine_04 {parent=spine_03, loc=<8.1024 0 0>, rot=<0 0 -0.0132 -0.9999>}
 spine_05 {parent=spine_04, loc=<10.0146 0 0>, rot=<0 0 -0.0015 -1>}
 neck_01 {parent=spine_05, loc=<11.3138 0.1529 0>, rot=<0 0 -0.2003 -0.9797>}
 neck_02 {parent=neck_01, loc=<4.5627 0 0>, rot=<0 0 0.0504 -0.9987>}
 head {parent=neck_02, loc=<5.3914 0 0>, rot=<0 0 0.115 -0.9934>}
 dyn_hood {parent=head, loc=<8.3502 -0.3245 -0.1049>, rot=<0.8025 -0.5967 0 0>}
 attach_cape {parent=spine_05, loc=<9.017 -11.714 0>, rot=<0.0616 -0.7044 -0.0616 -0.7044>}
 cape {parent=attach_cape, type=Null, loc=<0 0 0>, rot=<0 0 0 1>}
 attach_backpack {parent=spine_05, loc=<-0.123 -13.473 0>, rot=<-0.0093 -0.707 0.0093 -0.707>}
 Backpack_BR {parent=attach_backpack, type=Null, loc=<0 0 0>, rot=<0 0 0 1>}
 clavicle_l {parent=spine_05, loc=<4.3575 -0.1466 -2.3223>, rot=<0.0465 -0.7289 -0.009 -0.6829>}
 upperarm_l {parent=clavicle_l, loc=<15.5111 0 0>, rot=<0.0226 -0.3234 -0.0088 -0.9459>}
 lowerarm_l {parent=upperarm_l, loc=<26.6514 0 0>, rot=<0.0391 -0.0099 -0.2453 -0.9686>}
 hand_l {parent=lowerarm_l, loc=<22.3491 0 0>, rot=<-0.7081 -0.0507 0.0975 -0.6975>}
 thumb_01_l {parent=hand_l, loc=<1.3118 -0.3191 -2.459>, rot=<0.6546 -0.4042 0.0664 -0.6354>}
 thumb_02_l {parent=thumb_01_l, loc=<5.9336 -0.0448 0.0401>, rot=<-0.0037 0.0213 0.298 -0.9543>}
 thumb_03_l {parent=thumb_02_l, loc=<2.2223 0.3682 -0.0907>, rot=<0.0118 -0.013 0.0066 -0.9998>}
 index_metacarpal_l {parent=hand_l, loc=<1.6535 0.0183 -2.96>, rot=<0.0158 -0.0124 -0.0173 -0.9996>}
 index_01_l {parent=index_metacarpal_l, loc=<8.3008 0.0357 -0.0524>, rot=<-0.0044 0.0485 0.0924 -0.9945>}
 index_02_l {parent=index_01_l, loc=<3.3587 -0.3004 0.0452>, rot=<-0.002 0.0027 0.3225 -0.9466>}
 index_03_l {parent=index_02_l, loc=<2.5578 -0.123 -0.0252>, rot=<-0.002 0.0067 0.2558 -0.9667>}
 middle_metacarpal_l {parent=hand_l, loc=<2.1856 -0.1725 -0.5984>, rot=<0.006 0.0168 -0.0329 -0.9993>}
 middle_01_l {parent=middle_metacarpal_l, loc=<7.3275 -0.2447 0.0129>, rot=<-0.0208 0.0042 0.0858 -0.9961>}
 middle_02_l {parent=middle_01_l, loc=<4.0123 0.1401 0.1175>, rot=<-0.0218 0.0258 0.3512 -0.9357>}
 middle_03_l {parent=middle_02_l, loc=<2.6077 0.0534 0.1373>, rot=<0.0082 0.0188 0.1708 -0.9851>}
 ring_metacarpal_l {parent=hand_l, loc=<2.2074 -0.3625 1.3638>, rot=<-0.0284 0.0722 -0.0458 -0.9959>}
 ring_01_l {parent=ring_metacarpal_l, loc=<7.0295 -0.6197 -0.1446>, rot=<-0.0489 -0.0252 0.1892 -0.9804>}
 ring_02_l {parent=ring_01_l, loc=<3.6174 0.4647 -0.1027>, rot=<0 0 0.2806 -0.9598>}
 ring_03_l {parent=ring_02_l, loc=<2.4152 0.114 0.0199>, rot=<0 0 0.1189 -0.9929>}
 pinky_metacarpal_l {parent=hand_l, loc=<1.7616 -0.9753 3.3814>, rot=<-0.0397 0.0814 -0.069 -0.9935>}
 pinky_01_l {parent=pinky_metacarpal_l, loc=<7.1709 -0.4831 -0.1094>, rot=<-0.0654 -0.0035 0.2242 -0.9723>}
 pinky_02_l {parent=pinky_01_l, loc=<2.8274 0.0893 -0.1043>, rot=<0.0037 0.0011 0.2854 -0.9584>}
 pinky_03_l {parent=pinky_02_l, loc=<2.2046 0.0307 -0.0078>, rot=<-0.0002 0.0098 0.1445 -0.9895>}
 weapon_l {parent=hand_l, loc=<-0.1974 -3.4111 0.3811>, rot=<0 0 0 1>}
 hand_attach_l {parent=hand_l, loc=<8.5 -3 0>, rot=<0.7071 0.7071 0 0>}
 deform_wrist_dn_base_l {parent=hand_l, loc=<0 -2.2342 0>, rot=<0 0 0 -1>}
 deform_wrist_dn_l {parent=deform_wrist_dn_base_l, loc=<0 0 0>, rot=<0 0 0 1>}
 deform_wrist_up_base_l {parent=hand_l, loc=<0 2.2682 0>, rot=<0 0 0 -1>}
 deform_wrist_up_l {parent=deform_wrist_up_base_l, loc=<0 0 0>, rot=<0 0 0 1>}
 lowerarm_twist_02_l {parent=lowerarm_l, loc=<9.6199 0 0>, rot=<0 0 0 -1>}
 lowerarm_twist_01_l {parent=lowerarm_l, loc=<17.8549 0 0>, rot=<0 0 0 -1>}
 deform_elbow_in_base_l {parent=lowerarm_l, loc=<0.2112 5.259 0.8274>, rot=<0 0 0 -1>}
 deform_elbow_in_l {parent=deform_elbow_in_base_l, loc=<0 0 0>, rot=<0 0 0 1>}
 deform_elbow_out_base_l {parent=lowerarm_l, loc=<0 -4 0>, rot=<0 0 0 -1>}
 deform_elbow_out_l {parent=deform_elbow_out_base_l, loc=<0 0 0>, rot=<0 0 0 1>}
 MSTR_hand_cloth_2_l {parent=lowerarm_l, loc=<-5.0348 1.6515 -4.5412>, rot=<0.033 -0.4559 0.1532 -0.8761>}
 dyn_hand_cloth_21_l {parent=MSTR_hand_cloth_2_l, loc=<0 0 0>, rot=<0 0 0 1>}
 dyn_hand_cloth_22_l {parent=dyn_hand_cloth_21_l, loc=<5.5332 0 0>, rot=<-0.0404 -0.0301 0.0034 -0.9987>}
 dyn_hand_cloth_23_l {parent=dyn_hand_cloth_22_l, loc=<4.3149 0 0>, rot=<0.0133 -0.0906 -0.0018 -0.9958>}
 dyn_hand_cloth_24_l {parent=dyn_hand_cloth_23_l, loc=<3.2964 0 0>, rot=<0.0429 -0.1789 0.0078 -0.9829>}
 MSTR_hand_cloth_4_l {parent=lowerarm_l, loc=<-0.7512 0.3409 -3.9878>, rot=<0.026 -0.4801 0.1422 -0.8652>}
 dyn_hand_cloth_41_l {parent=MSTR_hand_cloth_4_l, loc=<0 0 0>, rot=<0 0 0 1>}
 dyn_hand_cloth_42_l {parent=dyn_hand_cloth_41_l, loc=<5.9639 0 0>, rot=<0.0017 -0.0137 -0.0243 -0.9996>}
 dyn_hand_cloth_43_l {parent=dyn_hand_cloth_42_l, loc=<6.3052 0 0>, rot=<-0.0013 0.0366 -0.0093 -0.9993>}
 dyn_hand_cloth_44_l {parent=dyn_hand_cloth_43_l, loc=<4.6603 0 0>, rot=<0.0041 -0.0911 -0.0064 -0.9958>}
 dyn_hand_cloth_45_l {parent=dyn_hand_cloth_44_l, loc=<3.7457 0 0>, rot=<0 0 0 1>}
 MSTR_hand_cloth_1_l {parent=lowerarm_l, loc=<4.4853 -0.6985 -3.3696>, rot=<0.0439 -0.528 0.0887 -0.8435>}
 dyn_hand_cloth_11_l {parent=MSTR_hand_cloth_1_l, loc=<0 0 0>, rot=<0 0 0 1>}
 dyn_hand_cloth_12_l {parent=dyn_hand_cloth_11_l, loc=<7.4973 0 0>, rot=<-0.0022 0.0644 -0.0053 -0.9979>}
 dyn_hand_cloth_13_l {parent=dyn_hand_cloth_12_l, loc=<4.7695 0 0>, rot=<0 -0.126 0 -0.992>}
 MSTR_hand_cloth_0_l {parent=lowerarm_l, loc=<10.426 -1.0303 -2.8959>, rot=<0.0349 -0.5174 0.0723 -0.852>}
 dyn_hand_cloth_01_l {parent=MSTR_hand_cloth_0_l, loc=<0 0 0>, rot=<0 0 0 1>}
 dyn_hand_cloth_02_l {parent=dyn_hand_cloth_01_l, loc=<6.6215 0 0>, rot=<0.0006 0.0482 -0.0244 -0.9985>}
 dyn_hand_cloth_03_l {parent=dyn_hand_cloth_02_l, loc=<6.5595 0 0>, rot=<0 0 0 1>}
 dyn_hand_cloth_04_l {parent=dyn_hand_cloth_03_l, loc=<5.9446 -0.3762 -1.3039>, rot=<0.0492 0.0688 -0.0078 -0.9964>}
 dyn_hand_cloth_05_l {parent=dyn_hand_cloth_04_l, loc=<3.7907 0 0>, rot=<0.0435 0.0673 -0.0029 -0.9968>}
 dyn_hand_cloth_06_l {parent=dyn_hand_cloth_03_l, loc=<6.3806 0.1902 1.3339>, rot=<0.0491 0.068 -0.0181 -0.9963>}
 dyn_hand_cloth_07_l {parent=dyn_hand_cloth_06_l, loc=<4.8128 0 0>, rot=<0.0434 0.0912 -0.004 -0.9949>}
 MSTR_hand_cloth_3_l {parent=lowerarm_l, loc=<18.3014 -0.5527 -2.8052>, rot=<0.1054 -0.6517 0.0202 -0.7509>}
 dyn_hand_cloth_31_l {parent=MSTR_hand_cloth_3_l, loc=<0 0 0>, rot=<0 0 0 1>}
 dyn_hand_cloth_32_l {parent=dyn_hand_cloth_31_l, loc=<3.5027 0 0>, rot=<0 0.1685 0 -0.9857>}
 upperarm_twist_01_l {parent=upperarm_l, loc=<0.0041 0 0>, rot=<0 0 0 -1>}
 deform_upperarm_bk_base_l {parent=upperarm_twist_01_l, loc=<1.0813 -7.2193 -1.9453>, rot=<0 0 0 1>}
 deform_upperarm_bk_l {parent=deform_upperarm_bk_base_l, loc=<0 0 0>, rot=<0 0 0 1>}
 deform_upperarm_fr_base_l {parent=upperarm_twist_01_l, loc=<-0.3691 7.0506 -2.765>, rot=<0 0 0 1>}
 deform_upperarm_fr_l {parent=deform_upperarm_fr_base_l, loc=<0 0 0>, rot=<0 0 0 1>}
 deform_upperarm_up_base_l {parent=upperarm_twist_01_l, loc=<-3.2216 -1.1391 3.3311>, rot=<0 0 0 1>}
 deform_upperarm_up_l {parent=deform_upperarm_up_base_l, loc=<0 0 0>, rot=<0 0 0 1>}
 upperarm_twist_02_l {parent=upperarm_l, loc=<13.7941 0 0>, rot=<0 0 0 -1>}
 clavicle_r {parent=spine_05, loc=<4.3575 -0.1466 2.3223>, rot=<0.7289 0.0465 -0.6829 0.009>}
 upperarm_r {parent=clavicle_r, loc=<-15.5111 0 0>, rot=<0.0226 -0.3234 -0.0088 -0.9459>}
 lowerarm_r {parent=upperarm_r, loc=<-26.6514 0 0>, rot=<0.0391 -0.0099 -0.2453 -0.9686>}
 hand_r {parent=lowerarm_r, loc=<-22.3491 0 0>, rot=<-0.7081 -0.0507 0.0975 -0.6975>}
 thumb_01_r {parent=hand_r, loc=<-1.3118 0.3191 2.459>, rot=<0.6546 -0.4042 0.0664 -0.6354>}
 thumb_02_r {parent=thumb_01_r, loc=<-5.9336 0.0448 -0.0401>, rot=<-0.0037 0.0213 0.298 -0.9543>}
 thumb_03_r {parent=thumb_02_r, loc=<-2.2223 -0.3682 0.0907>, rot=<0.0118 -0.013 0.0066 -0.9998>}
 index_metacarpal_r {parent=hand_r, loc=<-1.6535 -0.0183 2.96>, rot=<0.0158 -0.0124 -0.0173 -0.9996>}
 index_01_r {parent=index_metacarpal_r, loc=<-8.3008 -0.0357 0.0524>, rot=<-0.0044 0.0485 0.0924 -0.9945>}
 index_02_r {parent=index_01_r, loc=<-3.3587 0.3004 -0.0452>, rot=<-0.002 0.0027 0.3225 -0.9466>}
 index_03_r {parent=index_02_r, loc=<-2.5578 0.123 0.0252>, rot=<-0.002 0.0067 0.2558 -0.9667>}
 middle_metacarpal_r {parent=hand_r, loc=<-2.1856 0.1725 0.5984>, rot=<0.006 0.0168 -0.0329 -0.9993>}
 middle_01_r {parent=middle_metacarpal_r, loc=<-7.3275 0.2447 -0.0129>, rot=<-0.0208 0.0042 0.0858 -0.9961>}
 middle_02_r {parent=middle_01_r, loc=<-4.0123 -0.1401 -0.1175>, rot=<-0.0218 0.0258 0.3512 -0.9357>}
 middle_03_r {parent=middle_02_r, loc=<-2.6077 -0.0534 -0.1373>, rot=<0.0082 0.0188 0.1708 -0.9851>}
 ring_metacarpal_r {parent=hand_r, loc=<-2.2074 0.3625 -1.3638>, rot=<-0.0284 0.0722 -0.0458 -0.9959>}
 ring_01_r {parent=ring_metacarpal_r, loc=<-7.0295 0.6197 0.1446>, rot=<-0.0489 -0.0252 0.1892 -0.9804>}
 ring_02_r {parent=ring_01_r, loc=<-3.6174 -0.4647 0.1027>, rot=<0 0 0.2806 -0.9598>}
 ring_03_r {parent=ring_02_r, loc=<-2.4152 -0.114 -0.0199>, rot=<0 0 0.1189 -0.9929>}
 pinky_metacarpal_r {parent=hand_r, loc=<-1.7616 0.9753 -3.3814>, rot=<-0.0397 0.0814 -0.069 -0.9935>}
 pinky_01_r {parent=pinky_metacarpal_r, loc=<-7.1709 0.4831 0.1094>, rot=<-0.0654 -0.0035 0.2242 -0.9723>}
 pinky_02_r {parent=pinky_01_r, loc=<-2.8274 -0.0893 0.1043>, rot=<0.0037 0.0011 0.2854 -0.9584>}
 pinky_03_r {parent=pinky_02_r, loc=<-2.2046 -0.0307 0.0078>, rot=<-0.0002 0.0098 0.1445 -0.9895>}
 weapon_r {parent=hand_r, loc=<0.1974 3.4111 -0.3811>, rot=<0 0 0 1>}
 hand_attach_r {parent=hand_r, loc=<-8.5001 2.9999 0>, rot=<0 0 -0.7071 -0.7071>}
 deform_wrist_dn_base_r {parent=hand_r, loc=<-0.0004 2.2345 -0.0001>, rot=<0 0 0 -1>}
 deform_wrist_dn_r {parent=deform_wrist_dn_base_r, loc=<0 0 0>, rot=<0 0 0 1>}
 deform_wrist_up_base_r {parent=hand_r, loc=<0.0003 -2.2684 0.0001>, rot=<0 0 0 -1>}
 deform_wrist_up_r {parent=deform_wrist_up_base_r, loc=<0 0 0>, rot=<0 0 0 1>}
 lowerarm_twist_02_r {parent=lowerarm_r, loc=<-9.6199 0 0>, rot=<0 0 0 -1>}
 lowerarm_twist_01_r {parent=lowerarm_r, loc=<-17.8549 0 0>, rot=<0 0 0 -1>}
 deform_elbow_in_base_r {parent=lowerarm_r, loc=<-0.2111 -5.2591 -0.8275>, rot=<0 0 0 -1>}
 deform_elbow_in_r {parent=deform_elbow_in_base_r, loc=<-0.0001 0 0.0002>, rot=<0 0 0 1>}
 deform_elbow_out_base_r {parent=lowerarm_r, loc=<0 4 -0.0001>, rot=<0 0 0 -1>}
 deform_elbow_out_r {parent=deform_elbow_out_base_r, loc=<0 0 0>, rot=<0 0 0 1>}
 MSTR_hand_cloth_2_r {parent=lowerarm_r, loc=<4.7387 -1.9022 4.9466>, rot=<0.0507 -0.4196 0.1881 -0.8865>}
 dyn_hand_cloth_21_r {parent=MSTR_hand_cloth_2_r, loc=<0 0 0>, rot=<0 0 0 -1>}
 dyn_hand_cloth_22_r {parent=dyn_hand_cloth_21_r, loc=<-5.7717 0 0>, rot=<-0.0404 -0.0301 0.0034 -0.9987>}
 dyn_hand_cloth_23_r {parent=dyn_hand_cloth_22_r, loc=<-4.2548 0 0>, rot=<0.0493 -0.072 -0.0244 -0.9959>}
 dyn_hand_cloth_24_r {parent=dyn_hand_cloth_23_r, loc=<-3.2966 0 0>, rot=<0.0429 -0.1789 0.0078 -0.9829>}
 MSTR_hand_cloth_4_r {parent=lowerarm_r, loc=<0.751 -0.3409 3.988>, rot=<0.026 -0.4801 0.1422 -0.8652>}
 dyn_hand_cloth_41_r {parent=MSTR_hand_cloth_4_r, loc=<0 0 0>, rot=<0 0 0 1>}
 dyn_hand_cloth_42_r {parent=dyn_hand_cloth_41_r, loc=<-8.0699 0 0>, rot=<-0.0007 0.0144 -0.0199 -0.9997>}
 dyn_hand_cloth_43_r {parent=dyn_hand_cloth_42_r, loc=<-6.3446 0 0>, rot=<0.0169 -0.1214 -0.0368 -0.9918>}
 dyn_hand_cloth_44_r {parent=dyn_hand_cloth_43_r, loc=<-4.6603 0 0>, rot=<0.0041 -0.0911 -0.0064 -0.9958>}
 dyn_hand_cloth_45_r {parent=dyn_hand_cloth_44_r, loc=<-3.7458 0 0>, rot=<0 0 0 1>}
 MSTR_hand_cloth_1_r {parent=lowerarm_r, loc=<-6.5795 0.8425 3.8706>, rot=<0.0795 -0.5822 0.0652 -0.8065>}
 dyn_hand_cloth_11_r {parent=MSTR_hand_cloth_1_r, loc=<0 0 0>, rot=<0 0 0 1>}
 dyn_hand_cloth_12_r {parent=dyn_hand_cloth_11_r, loc=<-6.4066 0 0>, rot=<0 0 0 1>}
 dyn_hand_cloth_13_r {parent=dyn_hand_cloth_12_r, loc=<-4.7699 0 -0.0001>, rot=<0 -0.126 0 -0.992>}
 MSTR_hand_cloth_0_r {parent=lowerarm_r, loc=<-12.2977 1.1344 3.2765>, rot=<0.114 -0.5795 0.0389 -0.806>}
 dyn_hand_cloth_01_r {parent=MSTR_hand_cloth_0_r, loc=<0 0 0>, rot=<0 0 0 -1>}
 dyn_hand_cloth_02_r {parent=dyn_hand_cloth_01_r, loc=<-8.1633 0 0>, rot=<-0.0059 0.0311 0.0345 -0.9989>}
 dyn_hand_cloth_03_r {parent=dyn_hand_cloth_02_r, loc=<-6.5569 0 0>, rot=<0 0 0 1>}
 dyn_hand_cloth_04_r {parent=dyn_hand_cloth_03_r, loc=<-4.1083 -0.2289 0.7993>, rot=<0.0388 0.103 -0.016 -0.9938>}
 dyn_hand_cloth_05_r {parent=dyn_hand_cloth_04_r, loc=<-4.8969 0 0>, rot=<0 0 0 1>}
 dyn_hand_cloth_06_r {parent=dyn_hand_cloth_03_r, loc=<-2.2699 -0.1132 -1.825>, rot=<-0.0086 0.0298 0.0533 -0.9981>}
 dyn_hand_cloth_07_r {parent=dyn_hand_cloth_06_r, loc=<-4.8128 0 0>, rot=<0.0434 0.0912 -0.004 -0.9949>}
 MSTR_hand_cloth_3_r {parent=lowerarm_r, loc=<-18.3014 0.5527 2.8052>, rot=<0.1054 -0.6517 0.0202 -0.7509>}
 dyn_hand_cloth_31_r {parent=MSTR_hand_cloth_3_r, loc=<0 0 0>, rot=<0 0 0 1>}
 dyn_hand_cloth_32_r {parent=dyn_hand_cloth_31_r, loc=<-3.5025 -0.0001 0.0002>, rot=<0 0.1685 0 -0.9857>}
 upperarm_twist_01_r {parent=upperarm_r, loc=<-0.0041 0 0>, rot=<0 0 0 -1>}
 deform_upperarm_bk_base_r {parent=upperarm_twist_01_r, loc=<-1.0814 7.2192 1.9455>, rot=<0 0 0 1>}
 deform_upperarm_bk_r {parent=deform_upperarm_bk_base_r, loc=<0 0 0>, rot=<0 0 0 1>}
 deform_upperarm_fr_base_r {parent=upperarm_twist_01_r, loc=<0.3689 -7.0506 2.7652>, rot=<0 0 0 1>}
 deform_upperarm_fr_r {parent=deform_upperarm_fr_base_r, loc=<0 0 0>, rot=<0 0 0 1>}
 deform_upperarm_up_base_r {parent=upperarm_twist_01_r, loc=<3.2214 1.1391 -3.3308>, rot=<0 0 0 1>}
 deform_upperarm_up_r {parent=deform_upperarm_up_base_r, loc=<0 0 0>, rot=<0 0 0 -1>}
 upperarm_twist_02_r {parent=upperarm_r, loc=<-13.7941 0 0>, rot=<0 0 0 -1>}
 deform_pec_aim_base_l {parent=spine_04, loc=<7.8732 12.6388 -2.4123>, rot=<0.0102 -0.4602 0.3979 -0.7936>}
 deform_pec_aim_l {parent=deform_pec_aim_base_l, loc=<0 0 0>, rot=<0 0 0 1>}
 deform_pec_base_l {parent=deform_pec_aim_l, loc=<10.766 0 0>, rot=<0 0 0 1>}
 deform_pec_l {parent=deform_pec_base_l, loc=<0.055 0 0>, rot=<0 0 0 1>}
 deform_pec_aim_base_r {parent=spine_04, loc=<7.8727 12.6388 2.4123>, rot=<-0.4602 -0.0102 0.7936 0.3979>}
 deform_pec_aim_r {parent=deform_pec_aim_base_r, loc=<0 0 0>, rot=<0 0 0 1>}
 deform_pec_base_r {parent=deform_pec_aim_r, loc=<-10.7662 -0.0006 -0.0007>, rot=<0 0 0 1>}
 deform_pec_r {parent=deform_pec_base_r, loc=<0 0 0>, rot=<0 0 0 1>}
 thigh_l {parent=pelvis, loc=<-2.5693 -0.0109 -7.7501>, rot=<0.0012 0.0572 0.0155 -0.9982>}
 calf_l {parent=thigh_l, loc=<-39.0484 0 0>, rot=<0.0006 -0.0161 -0.0391 -0.9991>}
 foot_l {parent=calf_l, loc=<-42.561 0.1077 0.4292>, rot=<0.0418 -0.0455 0.0343 -0.9975>}
 ball_l {parent=foot_l, loc=<-6.1599 13.8264 0.6806>, rot=<0.0001 -0.0001 -0.7063 -0.7079>}
 calf_twist_02_l {parent=calf_l, loc=<-18.8064 -0.0967 0.0397>, rot=<0.0237 0.0095 -0.0035 -0.9997>}
 calf_twist_01_l {parent=calf_l, loc=<-35.813 -0.0967 0.0397>, rot=<0.0237 0.0095 -0.0035 -0.9997>}
 mstr_ankle_bk_l {parent=calf_twist_01_l, loc=<-4.3431 -3.6238 0.4427>, rot=<0.0177 -0.0552 0.7363 -0.6742>}
 def_ankle_bk_l {parent=mstr_ankle_bk_l, loc=<0 0 0>, rot=<0 0 0 1>}
 deform_knee_in_base_l {parent=calf_l, loc=<1.298 -5.2957 -0.0496>, rot=<0 0 0 -1>}
 deform_knee_in_l {parent=deform_knee_in_base_l, loc=<0 0 0>, rot=<0 0 0 1>}
 deform_knee_out_base_l {parent=calf_l, loc=<0 4 0>, rot=<0 0 0 -1>}
 deform_knee_out_l {parent=deform_knee_out_base_l, loc=<0 0 0>, rot=<0 0 0 1>}
 thigh_twist_01_l {parent=thigh_l, loc=<-21.6303 0 0>, rot=<0.0013 0 0 -1>}
 deform_groin_base_l {parent=thigh_l, loc=<-2.1908 8.1142 -1.7442>, rot=<0 0 0 -1>}
 deform_groin_l {parent=deform_groin_base_l, loc=<0 0 0>, rot=<0 0 0 1>}
 thigh_r {parent=pelvis, loc=<-2.5693 -0.0109 7.7501>, rot=<0.0572 -0.0012 0.9982 0.0155>}
 calf_r {parent=thigh_r, loc=<39.0484 0 0>, rot=<0.0006 -0.0161 -0.0391 -0.9991>}
 foot_r {parent=calf_r, loc=<42.561 -0.1077 -0.4292>, rot=<0.0418 -0.0455 0.0343 -0.9975>}
 ball_r {parent=foot_r, loc=<6.1599 -13.8264 -0.6806>, rot=<0.0001 -0.0001 -0.7063 -0.7079>}
 calf_twist_02_r {parent=calf_r, loc=<18.8064 0.0967 -0.0397>, rot=<0.0237 0.0095 -0.0035 -0.9997>}
 calf_twist_01_r {parent=calf_r, loc=<35.813 0.0967 -0.0397>, rot=<0.0237 0.0095 -0.0035 -0.9997>}
 mstr_ankle_bk_r {parent=calf_twist_01_r, loc=<4.3431 3.6238 -0.4426>, rot=<0.0177 -0.0552 0.7363 -0.6742>}
 def_ankle_bk_r {parent=mstr_ankle_bk_r, loc=<0 0 0>, rot=<0 0 0 1>}
 deform_knee_in_base_r {parent=calf_r, loc=<-1.298 5.2957 0.0496>, rot=<0 0 0 -1>}
 deform_knee_in_r {parent=deform_knee_in_base_r, loc=<0 0 0>, rot=<0 0 0 1>}
 deform_knee_out_base_r {parent=calf_r, loc=<0 -4 0>, rot=<0 0 0 -1>}
 deform_knee_out_r {parent=deform_knee_out_base_r, loc=<0 0 0>, rot=<0 0 0 1>}
 thigh_twist_01_r {parent=thigh_r, loc=<21.6303 0 0>, rot=<0.0013 0 0 -1>}
 deform_groin_base_r {parent=thigh_r, loc=<2.1908 -8.1142 1.7442>, rot=<0 0 0 -1>}
 deform_groin_r {parent=deform_groin_base_r, loc=<0 0 0>, rot=<0 0 0 1>}
 deform_glute_base_l {parent=pelvis, loc=<-8.8064 -8.8222 -8.4287>, rot=<0.0275 0.0501 -0.4472 -0.8926>}
 deform_glute_l {parent=deform_glute_base_l, loc=<0 0 0>, rot=<0 0 0 1>}
 deform_glute_base_r {parent=pelvis, loc=<-8.8062 -8.8222 8.4287>, rot=<0.0501 -0.0275 0.8926 -0.4472>}
 deform_glute_r {parent=deform_glute_base_r, loc=<0 0 0>, rot=<0 0 0 1>}
 MSTR_skirt_bk {parent=pelvis, loc=<-1.0849 -13.5823 0.0011>, rot=<-0.0027 -1 0.0001 -0.0026>}
 dyn_skirt_00_bk {parent=MSTR_skirt_bk, loc=<0 0 0>, rot=<0 0 0 1>}
 dyn_skirt_01_bk {parent=dyn_skirt_00_bk, loc=<17.6587 0 0>, rot=<0.0001 -0.0041 0.003 -1>}
 dyn_skirt_02_bk {parent=dyn_skirt_01_bk, loc=<13.9087 0 0>, rot=<-0.0001 0.0016 0.0133 -0.9999>}
 dyn_skirt_03_bk {parent=dyn_skirt_02_bk, loc=<10.4706 0 0>, rot=<0 0 0.011 -0.9999>}
 dyn_skirt_04_bk {parent=dyn_skirt_03_bk, loc=<10.535 0 0>, rot=<0 0 0 1>}
 MSTR_skirt_fw {parent=pelvis, loc=<-0.2034 11.0187 0>, rot=<0.0221 0.7068 0.7068 0.0221>}
 dyn_skirt_00_fw {parent=MSTR_skirt_fw, loc=<0 0 0>, rot=<0.7071 0 0 -0.7071>}
 dyn_skirt_01_fw {parent=dyn_skirt_00_fw, loc=<17.0596 0 0>, rot=<0 0 0.0015 -1>}
 dyn_skirt_02_fw {parent=dyn_skirt_01_fw, loc=<14.6962 0 0>, rot=<0 0 0 1>}
 dyn_skirt_03_fw {parent=dyn_skirt_02_fw, loc=<9.669 0 0>, rot=<0 0 0 1>}
 dyn_skirt_04_fw {parent=dyn_skirt_03_fw, loc=<10.4086 0 0>, rot=<0 0 -0.0196 -0.9998>}
 MSTR_skirt_fw_l {parent=pelvis, loc=<-0.2964 8.4291 -12.3127>, rot=<-0.0085 0.363 0.9313 0.0275>}
 dyn_skirt_00_fw_l {parent=MSTR_skirt_fw_l, loc=<0 0 0>, rot=<0 0 0 1>}
 dyn_skirt_01_fw_l {parent=dyn_skirt_00_fw_l, loc=<17.0683 0 0>, rot=<0 0 0.0185 -0.9998>}
 dyn_skirt_02_fw_l {parent=dyn_skirt_01_fw_l, loc=<14.7238 0 0>, rot=<0 0 0 1>}
 dyn_skirt_03_fw_l {parent=dyn_skirt_02_fw_l, loc=<9.2381 0 0>, rot=<0 0 0 1>}
 dyn_skirt_04_fw_l {parent=dyn_skirt_03_fw_l, loc=<10.8743 0 0>, rot=<0 0 0 1>}
 MSTR_skirt_side_l {parent=pelvis, loc=<-0.6569 -1.5567 -16.9496>, rot=<-0.0185 0.6767 0.7346 0.0464>}
 dyn_skirt_00_side_l {parent=MSTR_skirt_side_l, loc=<0 0 0>, rot=<0 0 0 1>}
 dyn_skirt_01_side_l {parent=dyn_skirt_00_side_l, loc=<17.1251 0 0>, rot=<-0.0341 -0.0034 0.0086 -0.9994>}
 dyn_skirt_02_side_l {parent=dyn_skirt_01_side_l, loc=<14.7717 0 0>, rot=<-0.023 -0.0027 0.0095 -0.9997>}
 dyn_skirt_03_side_l {parent=dyn_skirt_02_side_l, loc=<9.3951 0 0>, rot=<-0.0447 -0.0061 0.0094 -0.9989>}
 dyn_skirt_04_side_l {parent=dyn_skirt_03_side_l, loc=<10.7167 0 0>, rot=<0 0 -0.0533 -0.9986>}
 MSTR_skirt_bk_l {parent=pelvis, loc=<-0.9854 -10.7896 -13.0099>, rot=<-0.0056 -0.9762 -0.216 -0.0153>}
 dyn_skirt_00_bk_l {parent=MSTR_skirt_bk_l, loc=<0 0 0>, rot=<0 0 0 1>}
 dyn_skirt_01_bk_l {parent=dyn_skirt_00_bk_l, loc=<17.2521 0 0>, rot=<-0.0007 0.0142 0.0223 -0.9997>}
 dyn_skirt_02_bk_l {parent=dyn_skirt_01_bk_l, loc=<13.8377 0 0>, rot=<-0.0005 0.0062 0.0084 -0.9999>}
 dyn_skirt_03_bk_l {parent=dyn_skirt_02_bk_l, loc=<10.5097 0 0>, rot=<-0.0003 0.0061 -0.0006 -1>}
 dyn_skirt_04_bk_l {parent=dyn_skirt_03_bk_l, loc=<10.7004 0 0>, rot=<0.003 -0.0297 -0.0001 -0.9996>}
 MSTR_skirt_fw_r {parent=pelvis, loc=<-0.2964 8.4291 12.3127>, rot=<-0.363 -0.0085 0.0275 -0.9313>}
 dyn_skirt_00_fw_r {parent=MSTR_skirt_fw_r, loc=<0 0 0>, rot=<0 0 0 1>}
 dyn_skirt_01_fw_r {parent=dyn_skirt_00_fw_r, loc=<-17.0683 0 0>, rot=<0 0 0.0185 -0.9998>}
 dyn_skirt_02_fw_r {parent=dyn_skirt_01_fw_r, loc=<-14.7237 0 0>, rot=<0 0 0 1>}
 dyn_skirt_03_fw_r {parent=dyn_skirt_02_fw_r, loc=<-9.2381 0 0>, rot=<0 0 0 1>}
 dyn_skirt_04_fw_r {parent=dyn_skirt_03_fw_r, loc=<-10.8742 0 0>, rot=<0 0 0 1>}
 MSTR_skirt_side_r {parent=pelvis, loc=<-0.6569 -1.5567 16.9496>, rot=<-0.6767 -0.0185 0.0464 -0.7346>}
 dyn_skirt_00_side_r {parent=MSTR_skirt_side_r, loc=<0 0 0>, rot=<0 0 0 1>}
 dyn_skirt_01_side_r {parent=dyn_skirt_00_side_r, loc=<-17.1251 0 0>, rot=<-0.0341 -0.0034 0.0086 -0.9994>}
 dyn_skirt_02_side_r {parent=dyn_skirt_01_side_r, loc=<-14.7716 0.0001 0>, rot=<-0.023 -0.0027 0.0095 -0.9997>}
 dyn_skirt_03_side_r {parent=dyn_skirt_02_side_r, loc=<-9.3951 -0.0001 0>, rot=<-0.0447 -0.0061 0.0094 -0.9989>}
 dyn_skirt_04_side_r {parent=dyn_skirt_03_side_r, loc=<-10.7167 0 0>, rot=<0 0 -0.0533 -0.9986>}
 MSTR_skirt_bk_r {parent=pelvis, loc=<-0.9854 -10.7896 13.0099>, rot=<0.9762 -0.0056 -0.0153 0.216>}
 dyn_skirt_00_bk_r {parent=MSTR_skirt_bk_r, loc=<0 0 0>, rot=<0 0 0 1>}
 dyn_skirt_01_bk_r {parent=dyn_skirt_00_bk_r, loc=<-17.2522 0 0>, rot=<-0.0007 0.0142 0.0223 -0.9997>}
 dyn_skirt_02_bk_r {parent=dyn_skirt_01_bk_r, loc=<-13.8377 0 0>, rot=<-0.0005 0.0062 0.0084 -0.9999>}
 dyn_skirt_03_bk_r {parent=dyn_skirt_02_bk_r, loc=<-10.5097 -0.0001 0>, rot=<-0.0003 0.0061 -0.0006 -1>}
 dyn_skirt_04_bk_r {parent=dyn_skirt_03_bk_r, loc=<-10.7003 0 -0.0001>, rot=<0.003 -0.0297 -0.0001 -0.9996>}
 dyn_front_skirt_l_MSTR {parent=pelvis, loc=<7.4427 10.951 -5.5023>, rot=<-0.0036 0.0516 0.9984 0.0205>}
 dyn_front_skirt_l_01 {parent=dyn_front_skirt_l_MSTR, loc=<0 0 0>, rot=<0 0 0 1>}
 dyn_front_skirt_l_02 {parent=dyn_front_skirt_l_01, loc=<28.4449 0.0451 0.069>, rot=<0 0 0 1>}
 dyn_front_skirt_l_03 {parent=dyn_front_skirt_l_02, loc=<21.521 0.0387 0.0592>, rot=<0 0 0 1>}
 dyn_front_skirt_l_04 {parent=dyn_front_skirt_l_03, loc=<13.2954 0 0>, rot=<0 0 0 1>}
 dyn_front_skirt_l_05 {parent=dyn_front_skirt_l_04, loc=<12.098 0 0>, rot=<0 0 0 1>}
 dyn_front_skirt_corner_l_MSTR {parent=pelvis, loc=<7.6105 7.8246 -13.5315>, rot=<-0.022 0.3672 0.9294 0.03>}
 dyn_front_skirt_corner_l_01 {parent=dyn_front_skirt_corner_l_MSTR, loc=<0 0 0>, rot=<0 0 0 1>}
 dyn_front_skirt_corner_l_02 {parent=dyn_front_skirt_corner_l_01, loc=<28.4296 0.3627 0.343>, rot=<0 0 0 1>}
 dyn_front_skirt_corner_l_03 {parent=dyn_front_skirt_corner_l_02, loc=<21.5079 0.3113 0.2944>, rot=<0 0 0 1>}
 dyn_front_skirt_corner_l_04 {parent=dyn_front_skirt_corner_l_03, loc=<13.2954 0 0>, rot=<0 0 0 1>}
 dyn_front_skirt_corner_l_05 {parent=dyn_front_skirt_corner_l_04, loc=<12.098 0 0>, rot=<0 0 0 1>}
 dyn_rear_skirt_l_MSTR {parent=pelvis, loc=<9.572 -12.8098 -4.2043>, rot=<0.0177 -0.9986 -0.0482 -0.0105>}
 dyn_rear_skirt_l_01 {parent=dyn_rear_skirt_l_MSTR, loc=<0 0 0>, rot=<0 0 0 1>}
 dyn_rear_skirt_l_02_MSTR {parent=dyn_rear_skirt_l_01, loc=<28.4285 0.5685 -0.1247>, rot=<0 0 0 1>}
 dyn_rear_skirt_l_02 {parent=dyn_rear_skirt_l_02_MSTR, loc=<0 0 0>, rot=<0 0 0 1>}
 dyn_rear_skirt_l_03 {parent=dyn_rear_skirt_l_02, loc=<21.4984 0.4874 -0.1069>, rot=<0 0 0 1>}
 dyn_rear_skirt_l_04 {parent=dyn_rear_skirt_l_03, loc=<13.2954 0 0>, rot=<0 0 0 1>}
 dyn_rear_skirt_l_05 {parent=dyn_rear_skirt_l_04, loc=<12.098 0 0>, rot=<0 0 0 1>}
 dyn_rear_skirt_corner_l_MSTR {parent=pelvis, loc=<9.7909 -10.5335 -12.8306>, rot=<0.0247 -0.9718 -0.2332 -0.0233>}
 dyn_rear_skirt_corner_l_01 {parent=dyn_rear_skirt_corner_l_MSTR, loc=<0 0 0>, rot=<0 0 0 1>}
 dyn_rear_skirt_corner_l_02_MSTR {parent=dyn_rear_skirt_corner_l_01, loc=<28.4101 0.7173 -0.1384>, rot=<0 0 0 1>}
 dyn_rear_skirt_corner_l_02 {parent=dyn_rear_skirt_corner_l_02_MSTR, loc=<0 0 0>, rot=<0 0 0 1>}
 dyn_rear_skirt_corner_l_03 {parent=dyn_rear_skirt_corner_l_02, loc=<21.4938 0.6159 -0.1189>, rot=<0 0 0 1>}
 dyn_rear_skirt_corner_l_04 {parent=dyn_rear_skirt_corner_l_03, loc=<13.2954 0 0>, rot=<0 0 0 1>}
 dyn_rear_skirt_corner_l_05 {parent=dyn_rear_skirt_corner_l_04, loc=<12.098 0 0>, rot=<0 0 0 1>}
 dyn_front_skirt_r_MSTR {parent=pelvis, loc=<7.43 10.6445 5.4706>, rot=<-0.0516 -0.0036 0.0205 -0.9984>}
 dyn_front_skirt_r_01 {parent=dyn_front_skirt_r_MSTR, loc=<0 0 0>, rot=<0 0 0 1>}
 dyn_front_skirt_r_02 {parent=dyn_front_skirt_r_01, loc=<-28.4449 -0.0451 -0.069>, rot=<0 0 0 1>}
 dyn_front_skirt_r_03 {parent=dyn_front_skirt_r_02, loc=<-21.521 -0.0387 -0.0592>, rot=<0 0 0 1>}
 dyn_front_skirt_r_04 {parent=dyn_front_skirt_r_03, loc=<-13.2955 0 0>, rot=<0 0 0 1>}
 dyn_front_skirt_r_05 {parent=dyn_front_skirt_r_04, loc=<-12.098 0 0>, rot=<0 0 0 1>}
 dyn_front_skirt_corner_r_MSTR {parent=pelvis, loc=<7.539 7.102 12.8559>, rot=<-0.3672 -0.022 0.03 -0.9294>}
 dyn_front_skirt_corner_r_01 {parent=dyn_front_skirt_corner_r_MSTR, loc=<0 0 0>, rot=<0 0 0 1>}
 dyn_front_skirt_corner_r_02 {parent=dyn_front_skirt_corner_r_01, loc=<-28.4296 -0.3627 -0.343>, rot=<0 0 0 1>}
 dyn_front_skirt_corner_r_03 {parent=dyn_front_skirt_corner_r_02, loc=<-21.5079 -0.3113 -0.2944>, rot=<0 0 0 1>}
 dyn_front_skirt_corner_r_04 {parent=dyn_front_skirt_corner_r_03, loc=<-13.2955 0 0>, rot=<0 0 0 1>}
 dyn_front_skirt_corner_r_05 {parent=dyn_front_skirt_corner_r_04, loc=<-12.098 0 0>, rot=<0 0 0 1>}
 dyn_rear_skirt_r_MSTR {parent=pelvis, loc=<9.5884 -13.2623 4.248>, rot=<0.9986 0.0177 -0.0105 0.0482>}
 dyn_rear_skirt_r_01 {parent=dyn_rear_skirt_r_MSTR, loc=<0 0 0>, rot=<0 0 0 1>}
 dyn_rear_skirt_r_02_MSTR {parent=dyn_rear_skirt_r_01, loc=<-28.4072 -0.567 0.1243>, rot=<0 0 0 1>}
 dyn_rear_skirt_r_02 {parent=dyn_rear_skirt_r_02_MSTR, loc=<0 0 0>, rot=<0 0 0 1>}
 dyn_rear_skirt_r_03 {parent=dyn_rear_skirt_r_02, loc=<-21.5197 -0.4889 0.1072>, rot=<0 0 0 1>}
 dyn_rear_skirt_r_04 {parent=dyn_rear_skirt_r_03, loc=<-13.2955 0 0>, rot=<0 0 0 1>}
 dyn_rear_skirt_r_05 {parent=dyn_rear_skirt_r_04, loc=<-12.0979 0 0>, rot=<0 0 0 1>}
 dyn_rear_skirt_corner_r_MSTR {parent=pelvis, loc=<9.8493 -11.4155 13.2787>, rot=<0.9718 0.0247 -0.0233 0.2332>}
 dyn_rear_skirt_corner_r_01 {parent=dyn_rear_skirt_corner_r_MSTR, loc=<0 0 0>, rot=<0 0 0 -1>}
 dyn_rear_skirt_corner_r_02_MSTR {parent=dyn_rear_skirt_corner_r_01, loc=<-28.4189 -0.7231 0.1449>, rot=<0 0 0 -1>}
 dyn_rear_skirt_corner_r_02 {parent=dyn_rear_skirt_corner_r_02_MSTR, loc=<0 0 0>, rot=<0 0 0 -1>}
 dyn_rear_skirt_corner_r_03 {parent=dyn_rear_skirt_corner_r_02, loc=<-21.4849 -0.6101 0.1124>, rot=<0 0 0 -1>}
 dyn_rear_skirt_corner_r_04 {parent=dyn_rear_skirt_corner_r_03, loc=<-13.2955 0 0>, rot=<0 0 0 -1>}
 dyn_rear_skirt_corner_r_05 {parent=dyn_rear_skirt_corner_r_04, loc=<-12.098 0.0001 0>, rot=<0 0 0 -1>}
 dyn_side_skirt_r_MSTR {parent=pelvis, loc=<8.7216 -1.6343 17.6738>, rot=<-0.7211 -0.0289 0.0301 -0.6916>}
 dyn_side_skirt_r_01 {parent=dyn_side_skirt_r_MSTR, loc=<0 0 0>, rot=<0 0 0 1>}
 dyn_side_skirt_r_02 {parent=dyn_side_skirt_r_01, loc=<-28.4128 -0.6695 -0.2554>, rot=<0 0 0 1>}
 dyn_side_skirt_r_03 {parent=dyn_side_skirt_r_02, loc=<-21.4935 -0.5748 -0.2192>, rot=<0 0 0 1>}
 dyn_side_skirt_r_04 {parent=dyn_side_skirt_r_03, loc=<-13.2954 0 0>, rot=<0 0 0 1>}
 dyn_side_skirt_r_05 {parent=dyn_side_skirt_r_04, loc=<-12.098 0 0>, rot=<0 0 0 1>}
 dyn_side_skirt_l_MSTR {parent=pelvis, loc=<8.7216 -1.6343 -17.6738>, rot=<0.0289 -0.7211 -0.6916 -0.0301>}
 dyn_side_skirt_l_01 {parent=dyn_side_skirt_l_MSTR, loc=<0 0 0>, rot=<0 0 0 1>}
 dyn_side_skirt_l_02 {parent=dyn_side_skirt_l_01, loc=<28.4128 0.6696 0.2554>, rot=<0 0 0 1>}
 dyn_side_skirt_l_03 {parent=dyn_side_skirt_l_02, loc=<21.4935 0.5747 0.2192>, rot=<0 0 0 1>}
 dyn_side_skirt_l_04 {parent=dyn_side_skirt_l_03, loc=<13.2954 0 0>, rot=<0 0 0 1>}
 dyn_side_skirt_l_05 {parent=dyn_side_skirt_l_04, loc=<12.098 0 0>, rot=<0 0 0 1>}
 dyn_root_offset {parent=root, loc=<0 0 0>, rot=<0 0 0 1>}
]
