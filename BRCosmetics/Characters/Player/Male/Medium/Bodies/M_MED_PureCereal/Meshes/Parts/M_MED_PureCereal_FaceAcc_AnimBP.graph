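# Animation Blueprint for the PureCereal male character's face accessory, adding secondary motion to hoodie bones over the pose copied from its attached parent.
# Weighted constraints tie each hoodie corner to spine, clavicle, neck, head and pectoral bones, followed by additive offsets.
# Spring-driven AnimDynamics then simulates both corners and the hood in root-relative space.
# The event graph derives left and right head-rotation alphas, which the AnimGraph never reads.
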AnimBlueprint /BRCosmetics/Characters/Player/Male/Medium/Bodies/M_MED_PureCereal/Meshes/Parts/M_MED_PureCereal_FaceAcc_AnimBP
target_skeleton=&/M_MED_PureCereal_FaceAcc_Skeleton
graphs [
 AnimGraph {nodes [
   n1 = "Constraint - dyn_hoodieCorner_l (PPPPPPPP)"!(
    ComponentPose=Local To Component(LocalPose=Copy Pose From Mesh() {bUseAttachedParent=true})
    Alpha=1
    ConstraintWeights_0=0.0094
    ConstraintWeights_1:float=0.0157
    ConstraintWeights_2:float=0.365
    ConstraintWeights_3:float=0.3161
    ConstraintWeights_4:float=0.0026
    ConstraintWeights_5:float=0.0229
    ConstraintWeights_6:float=0.2647
    ConstraintWeights_7:float=0.0035
   ) -> (Pose) {
    alphaBoolBlend {blendOption=Linear}
    boneToModify {boneName=dyn_hoodieCorner_l}
    constraintSetup [
     {targetBone {boneName=spine_03}}
     {targetBone {boneName=spine_04}}
     {targetBone {boneName=spine_05}}
     {targetBone {boneName=clavicle_l}}
     {targetBone {boneName=neck_01}}
     {targetBone {boneName=neck_02}}
     {targetBone {boneName=head}}
     {targetBone {boneName=deform_pec_l}}
    ]
    constraintWeights [1,1,1,1,1,1,1,1]
   }
   n2 = "Constraint - dyn_hoodieCorner_r (PPPPPPPP)"!(
    ComponentPose=@n1
    Alpha=1
    ConstraintWeights_0=0.0053
    ConstraintWeights_1:float=0.0021
    ConstraintWeights_2:float=0.4033
    ConstraintWeights_3:float=0.3343
    ConstraintWeights_4:float=0.0124
    ConstraintWeights_5:float=0.0595
    ConstraintWeights_6:float=0.1829
    ConstraintWeights_7:float=0.0002
   ) -> (Pose) {
    alphaBoolBlend {blendOption=Linear}
    boneToModify {boneName=dyn_hoodieCorner_r}
    constraintSetup [
     {targetBone {boneName=spine_03}}
     {targetBone {boneName=spine_04}}
     {targetBone {boneName=spine_05}}
     {targetBone {boneName=clavicle_r}}
     {targetBone {boneName=neck_01}}
     {targetBone {boneName=neck_02}}
     {targetBone {boneName=head}}
     {targetBone {boneName=deform_pec_r}}
    ]
    constraintWeights [1,1,1,1,1,1,1,1]
   }
   n3 = "Transform (Modify) Bone - Bone: dyn_hoodieCorner_l"(ComponentPose=@n2, Alpha=1, Translation="-3,-1,-3", Rotation="0,-30,5", Scale="1,1,1") {
    rotationMode=BMM_Additive
    translationMode=BMM_Additive
    alphaBoolBlend {blendOption=Linear}
    boneToModify {boneName=dyn_hoodieCorner_l}
    rotation {roll=5, yaw=-30}
    translation=<-3 -1 -3>
   }
   n4 = "Transform (Modify) Bone - Bone: dyn_hoodieCorner_r"(ComponentPose=@n3, Alpha=1, Translation="0,-1,-4", Rotation="0,30,5", Scale="1,1,1") {
    rotationMode=BMM_Additive
    translationMode=BMM_Additive
    alphaBoolBlend {blendOption=Linear}
    boneToModify {boneName=dyn_hoodieCorner_r}
    rotation {roll=5, yaw=30}
    translation {y=-1, z=-4}
   }
   n5 = Anim Dynamics - Bone: dyn_hoodieCorner_l(ComponentPose=@n4, Alpha=1, ExternalForce="0,0,0") {
    angularDampingOverride=0.999
    angularSpringConstant=300
    bAngularSpring=true
    bOverrideAngularDamping=true
    bOverrideLinearDamping=true
    bUseGravityOverride=true
    linearDampingOverride=0.9999
    simulationSpace=RootRelative
    alphaBoolBlend {blendOption=Linear}
    boundBone {boneName=dyn_hoodieCorner_l}
    componentAppliedLinearAccClamp=<5000 5000 5000>
    componentLinearAccScale=<0.5 0.5 0.5>
    componentLinearVelScale=<0.5 0.5 0.5>
    gravityOverride {z=-980}
    physicsBodyDefinitions [{
      boundBone {boneName=dyn_hoodieCorner_l}
      boxExtents=<13 13 13>
      constraintSetup {angularLimitsMax=<10 10 10>, angularLimitsMin=<-20 -10 -10>, angularTarget=<1 0 0>, linearAxesMax=<0 0 2>, linearAxesMin=<0 0 -0.5>}
      localJointOffset=<0 0 -1>
     }]
    simSpaceSettings {maxAngularAcceleration=3000, maxAngularVelocity=3000, simSpaceAngularAlpha=0.5}
   }
   n6 = Anim Dynamics - Bone: dyn_hoodieCorner_r(ComponentPose=@n5, Alpha=1, ExternalForce="0,0,0") {
    angularDampingOverride=0.999
    angularSpringConstant=300
    bAngularSpring=true
    bOverrideAngularDamping=true
    bOverrideLinearDamping=true
    bUseGravityOverride=true
    linearDampingOverride=0.9999
    simulationSpace=RootRelative
    alphaBoolBlend {blendOption=Linear}
    boundBone {boneName=dyn_hoodieCorner_r}
    componentAppliedLinearAccClamp=<5000 5000 5000>
    componentLinearAccScale=<0.5 0.5 0.5>
    componentLinearVelScale=<0.5 0.5 0.5>
    gravityOverride {z=-980}
    physicsBodyDefinitions [{
      boundBone {boneName=dyn_hoodieCorner_r}
      boxExtents=<13 13 13>
      constraintSetup {angularLimitsMax=<20 10 10>, angularLimitsMin=<-10 -10 -10>, angularTarget=<1 0 0>, linearAxesMax=<0 0 0.5>, linearAxesMin=<0 0 -2>}
      localJointOffset=<0 0 1>
     }]
    simSpaceSettings {maxAngularAcceleration=3000, maxAngularVelocity=3000, simSpaceAngularAlpha=0.5}
   }
   n7 = Anim Dynamics - Bone: dyn_hood(ComponentPose=@n6, Alpha=1, ExternalForce="0,0,0") {
    angularDampingOverride=0.999
    angularSpringConstant=600
    bAngularSpring=true
    bOverrideAngularDamping=true
    bUseGravityOverride=true
    simulationSpace=RootRelative
    alphaBoolBlend {blendOption=Linear}
    boundBone {boneName=dyn_hood}
    componentAppliedLinearAccClamp=<5000 5000 5000>
    componentLinearAccScale=<0.5 0.5 0.5>
    componentLinearVelScale=<0.5 0.5 0.5>
    gravityOverride {z=-980}
    physicsBodyDefinitions [{
      boundBone {boneName=dyn_hood}
      boxExtents=<12 12 12>
      constraintSetup {angularLimitsMax=<0 10 30>, angularLimitsMin=<0 -10 -45>, angularTarget=<1 0 0>}
      localJointOffset=<0.8 0 0>
     }]
    relativeSpaceBone {boneName=head}
    simSpaceSettings {maxAngularAcceleration=3000, maxAngularVelocity=3000, simSpaceAngularAlpha=0.2}
   }
   Output Pose(Result=Component To Local(ComponentPose=@n7)) {layerGroup=None}
  ]}
 EventGraph {nodes [
   Event BlueprintUpdateAnimation()
   n1 = GetPartSkeletalMeshComponent(self=Self-Reference())
   > Set HeadRotateAlpha_L(HeadRotateAlpha_L:float=GetAlphaFromReferenceTransformDelta(SkeletalMeshComponent=@n1, BoneName=spine_05, BaseName=head, DeltaComponentType=EulerX, InputRange=(X=-40,Y=-75), TwistAxis="1,0,0")) -> (Output_Get:float)
   > Set HeadRotateAlpha_R(HeadRotateAlpha_R:float=GetAlphaFromReferenceTransformDelta(
     SkeletalMeshComponent=@n1
     BoneName=spine_05
     BaseName=head
     DeltaComponentType=EulerX
     InputRange=(X=75,Y=40)
     OutputRange=(X=1,Y=0)
     TwistAxis="1,0,0"
    )) -> (Output_Get:float)
  ]}
]
variables [
 HeadRotateAlpha_L {type=number}
 HeadRotateAlpha_R {type=number}
]
